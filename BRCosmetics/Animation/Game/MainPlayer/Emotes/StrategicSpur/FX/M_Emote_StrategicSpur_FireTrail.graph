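# Translucent unlit fire trail material for the StrategicSpur emote, built for Niagara ribbons and mesh particles.
# Scrolling noise, weighted along U, distorts a panning base texture and a flame mask, which multiply and are power-shaped.
# The Switch_Orange parameter blends between a caustics-boosted result and the plain shape, defaulting to plain.
# Particle color feeds emissive adjustments, while particle alpha scales the shape, depth-faded except on forward and mobile paths.
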Material /BRCosmetics/Animation/Game/MainPlayer/Emotes/StrategicSpur/FX/M_Emote_StrategicSpur_FireTrail
graph {
 properties {
  bUsedWithNiagaraMeshParticles=true
  bUsedWithNiagaraRibbons=true
  blendMode=BLEND_Translucent
  pixelDepthOffsetMode=PDOM_Legacy
  shadingModel=MSM_Unlit
  usedShadingModels=Unlit
 }
 nodes [
  Constant() {r=0.25}
  n1 = ParticleColor()
  MaterialFunctionCall(Emissive=@n1.RGB) {materialFunction=&MF_FX_EmissiveAdjustments}
  n2 = VectorParameter() {parameterName=BaseUV, defaultValue=<rgba 2 1.5 0 0>}
  n3 = Add(
   A=Multiply(
    A=TextureCoordinate() {vTiling=0.8}
    B=MaterialFunctionCall(X=@n2.R, Y=@n2.R) {materialFunction=&MakeFloat2}.Result
   )
   B=Constant2Vector() {r=0.15}
  )
  n4 = TextureSample(UVs=Panner(Coordinate=TextureCoordinate() {uTiling=1.2, vTiling=0.8}) {speedX=-3}) {texture=/BRCosmetics/Weapons/FORT_Melee/Pickaxe_CoatCheck/FX/T_Pickaxe_CoatCheck_Noise}
  n5 = ComponentMask(None=TextureCoordinate()) {r=true}
  n6 = Multiply(
   A=Multiply(A=ComponentMask(None=@n4.RGB) {g=true, r=true}, B=LinearInterpolate(Alpha=@n5) {constA=0.1})
   B=ScalarParameter() {defaultValue=0.2, parameterName=Noise_Intensity}
  ) {constB=0.3}
  n7 = ScalarParameter() {defaultValue=-4, parameterName=BaseTexture_SpeedX}
  n8 = ScalarParameter() {defaultValue=-0.1, parameterName=BaseTexture_SpeedY}
  n9 = TextureSampleParameter2D(UVs=Panner(Coordinate=Add(A=@n3, B=@n6), Speed=AppendVector(A=@n7, B=@n8)) {speedX=-1, speedY=-0.1}) {parameterName=Base_Texture, samplerType=SAMPLERTYPE_LinearColor, texture=&T_NoisePack_LinearColor}
  n10 = Multiply(
   A=ChannelMaskParameter(None=@n9.RGBA) {parameterName=Tex_ChargeRGBA, sortPriority=7}.R
   B=ScalarParameter() {defaultValue=1.3, parameterName=NoiseMult}
  )
  n11 = ScalarParameter() {defaultValue=1, parameterName=InitPower}
  n12 = ScalarParameter() {defaultValue=1, parameterName=MaskTextureU}
  n13 = ScalarParameter() {defaultValue=1, parameterName=MaskTextureV}
  n14 = Add(
   A=Multiply(A=TextureCoordinate(), B=AppendVector(A=@n12, B=@n13))
   B=LinearInterpolate(A=ScalarParameter() {parameterName=MinOpacity}, Alpha=@n6) {constA=0.1}
  )
  n15 = AppendVector(A=ScalarParameter() {defaultValue=-2.2, parameterName=MaskTexture_SpeedX}, B=ScalarParameter() {parameterName=MaskTexture_SpeedY})
  n16 = Multiply(
   A=LinearInterpolate(A=ScalarParameter() {defaultValue=0.2, parameterName=Param_3}, Alpha=Power(Base=@n10, Exp=@n11)) {constA=0.1}
   B=TextureSampleParameter2D(UVs=Panner(Coordinate=@n14, Speed=@n15) {speedX=-1}) {parameterName=MaskTexture, texture=/Game/Accessories/FORT_Backpacks/Backpack_M_MED_KeyChain/FX/T_Backpack_KeyChain_Flame}.R
  )
  n17 = Power(
   Base=LinearInterpolate(B=ScalarParameter() {defaultValue=0.8, parameterName=LerpUp}, Alpha=@n16) {constB=2.5}
   Exp=ScalarParameter() {defaultValue=1, parameterName=Power}
  )
  n18 = TextureCoordinate() {uTiling=0.6, vTiling=0.4}
  n19 = VectorParameter() {parameterName=CausticsUV, defaultValue=<rgba 1.5 1 0 0>}
  n20 = MaterialFunctionCall(X=@n19.R, Y=@n19.R) {materialFunction=&MakeFloat2}
  n21 = TextureSampleParameter2D(UVs=Panner(Coordinate=Multiply(A=@n18, B=@n20.Result)) {speedX=-3, speedY=-0.05}) {parameterName=Orange_Texture, texture=/Game/Global/Textures/Default/Caustics/T_Caustics_Softer}
  n22 = ScalarParameter() {defaultValue=6, parameterName=CausticsMult}
  n23 = Add(A=@n16, B=Multiply(A=@n17, B=Multiply(A=@n21.R, B=@n22) {constB=5}))
  n24 = ScalarParameter() {defaultValue=1, parameterName=Switch_Orange}
  n25 = Multiply(A=Multiply(A=@n1.A, B=LinearInterpolate(A=@n23, B=@n17, Alpha=@n24)))
  n26 = DepthFade(Opacity=@n25, FadeDistance=ScalarParameter() {defaultValue=5, parameterName=DepthFade})
  ShadingPathSwitch(Default=@n26, Deferred=@n26, Forward=@n25, Mobile=@n25)
 ]
}
