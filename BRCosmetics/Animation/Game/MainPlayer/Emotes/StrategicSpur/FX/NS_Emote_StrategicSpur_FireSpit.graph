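# Niagara system for the fire-spit effect in the Strategic Spur emote, a flame jet along the owner's Y axis.
# Ribbon and sprite flames, an initial burst and glow, embers, mesh fuel spit and drips, and a light form the effect.
# User.Reveal drives jet length, opacity, brightness, ember spawning and light.
# Flame positions blend owner rotation toward a moving-average quaternion along the length, so the jet bends with lag when turned.
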NiagaraSystem /BRCosmetics/Animation/Game/MainPlayer/Emotes/StrategicSpur/FX/NS_Emote_StrategicSpur_FireSpit
dependencies {
 usedDataInterfaces [&NiagaraDataInterfaceCurve, &NiagaraDataInterfaceColorCurve]
 usedDynamicInputs [&SimulationPosition, &FNTimeOfDayScalar]
 usedModules [
  &self:SystemSpawnScript.SystemScriptSource.SystemScriptGraph.NiagaraNodeAssignment_0.SetVariables_#1
  &SystemState
  &self:SystemSpawnScript.SystemScriptSource.SystemScriptGraph.NiagaraNodeAssignment_1.SetVariables_#2
  &InterpolateOverTime
 ]
}
emitters {
 AdditionalFlames001 {
  emitterSpawnScript [/SetVariables(Emitter.Length=166, Emitter.ParticlesToSpawn=16) {isSetParameters=true, module=SetVariables_#3}]
  emitterUpdateScript [
   &EmitterState(Life Cycle Mode=::NewEnumerator1, Loop Behavior=::NewEnumerator1, Loop Duration=1, UseLoopDelay=false)
   &SpawnBurst_Instantaneous(Spawn Count=$Emitter.ParticlesToSpawn, Spawn Probability=0.5, Use Spawn Probability=false, UseLoopCountLimit=false)
   /SetVariables(Emitter.Reveal:NiagaraFloat=$User.Reveal) {isSetParameters=true, module=SetVariables_#4}
   /SetVariables(Emitter.Length:NiagaraFloat=&Multiply_Float(
     A=$Emitter.Initial.Length
     B=&FloatFromCurve(
      FloatCurve=DataInterface() {propertyValues {Curve {keys [{interpMode=RCIM_Cubic, leaveTangent=2.4356, tangentMode=RCTM_Break, time=0, value=0}, <at 1 1>]}}}
      CurveIndex=$Emitter.Reveal
     )
    )) {isSetParameters=true, module=SetVariables_#5}
  ]
  particleSpawnScript [&InitializeParticle(
    Color=<rgba 3 3 3 1>
    Sprite Rotation Angle=&UniformRangedFloat(Minimum=-30, Maximum=30)
    Sprite Rotation Mode=::NewEnumerator0
    Sprite Size=&Multiply_Vector2D_ByFloat(Float=&RandomRangeFloat(Minimum=50, Maximum=50), Vector2D=<0.2 2>)
    Sprite Size Mode=::NewEnumerator4
    Sprite UV Mode=::NewEnumerator1
    UsePositionOffset=false
    Write Material Random=true
   )]
  particleUpdateScript [
   &ParticleState(Kill Particles When Lifetime Has Elapsed=false)
   /SetVariables(Particles.ParticleRelativeLifetimeSprites:NiagaraFloat=&Frac_Float(Float=&Add_Float(A=&Multiply_Float(A=1.5, B=$Emitter.Age), B=&ReturnNormalizedExecIndex()))) {isSetParameters=true, module=SetVariables_#6}
   /SetVariables(Particles.OpacityMultiplierSprites:NiagaraFloat=&FloatFromCurve(
     FloatCurve=DataInterface() {propertyValues {Curve {keys [
         {interpMode=RCIM_Cubic, tangentMode=RCTM_User, time=0, value=0}
         {interpMode=RCIM_Cubic, tangentMode=RCTM_User, time=0.5, value=1}
         {tangentMode=RCTM_User, time=1, value=0}
        ]}}}
     CurveIndex=&Add_Float(
      A=&Multiply_Float(A=$Particles.ParticleRelativeLifetimeSprites, B=-1)
      B=&Multiply_Float(A=$Emitter.Reveal, B=2)
     )
    )) {isSetParameters=true, module=SetVariables_#7}
   /SetVariables(Particles.InterpolateCurveSprites:NiagaraFloat=&FloatFromCurve(
     FloatCurve=DataInterface() {propertyValues {Curve {keys [
         {arriveTangent=-0.6503, interpMode=RCIM_Cubic, leaveTangent=1.4736, tangentMode=RCTM_Break, time=0, value=0}
         {arriveTangent=0.4202, interpMode=RCIM_Cubic, tangentMode=RCTM_Break, time=1, value=1}
        ]}}}
     CurveIndex=$Particles.ParticleRelativeLifetimeSprites
    )) {isSetParameters=true, module=SetVariables_#8}
   /SetVariables(Particles.PositionSprites:NiagaraPosition=&AddVectorToPosition(
     Position=&AddVectorToPosition(
      Position=/Niagara/DynamicInputs/Helpers/SimulationPosition()
      Vector=&TransformVector(Vector=<0 1 0>, Destination Space=::Simulation)
     )
     Vector=&Multiply_VectorByFloat(
      Vector=&RotateVectorByQuaternion(
       Vector To Rotate=<0 1 0>
       Quaternion=&LerpQuaternion(
        Quaternion A:Quat4f=$Engine.Owner.Rotation
        Quaternion B:Quat4f=$System.InterpolateOverTime.InterpolateOverTime_Quaternion.Moving Average
        Lerp Factor:NiagaraFloat=$Particles.InterpolateCurveSprites
        Enable Spherical Linear Interpolation:NiagaraBool=false
       )
      )
      Float=&Multiply_Float(A=$Emitter.Length, B=$Particles.InterpolateCurveSprites)
     )
    )) {isSetParameters=true, module=SetVariables_#9}
   /SetVariables(Particles.SpriteSize=&Multiply_Vector2D(
     A=&Vector2DFromFloat(Value=&FloatFromCurve(
       FloatCurve=DataInterface() {propertyValues {Curve {keys [
           {arriveTangent=0.5131, interpMode=RCIM_Cubic, leaveTangent=0.1835, tangentMode=RCTM_Break, time=0, value=0.4}
           {interpMode=RCIM_Cubic, tangentMode=RCTM_Break, time=0.6, value=1}
           {arriveTangent=-3.5982, tangentMode=RCTM_User, time=1, value=0.5}
          ]}}}
       CurveIndex=$Particles.InterpolateCurveSprites
       Scale Curve=&Lerp_Float(A=55, B=70, Alpha=&RandomRangeFloat(Minimum=0, Maximum=1))
      ))
     B=<1 1>
    )) {isSetParameters=true, module=SetVariables_#10}
   &SubUVAnimation(SubUV Lookup Index:NiagaraFloat=$Particles.ParticleRelativeLifetimeSprites)
   &ScaleColor(
    Scale Alpha=&Multiply_Float(
     A=&Multiply_Float(
      A=&FloatFromCurve(
       FloatCurve=DataInterface() {propertyValues {Curve {keys [
           {arriveTangent=16.5612, interpMode=RCIM_Cubic, leaveTangentWeight=0.3333, time=0.1, value=0}
           {arriveTangentWeight=0.1333, interpMode=RCIM_Cubic, leaveTangentWeight=0.1667, tangentMode=RCTM_User, time=0.5, value=1}
           {arriveTangentWeight=0.1667, interpMode=RCIM_Cubic, tangentMode=RCTM_User, time=1, value=0}
          ]}}}
       CurveIndex=$Particles.ParticleRelativeLifetimeSprites
      )
      B=&Multiply_Float(
       A=&RandomRangeFloat(Minimum=1, Maximum=1)
       B=&UniformAOrBFloat(A=0, Distribution Weight=1, Evaluation Type=::NewEnumerator0, Recalculate Random Each Loop=true, Override Seed=false)
      )
     )
     B=$Particles.OpacityMultiplierSprites
    )
    Scale RGB=&Multiply_VectorByFloat(
     Vector=&Multiply_VectorByFloat(
      Vector=<1 1 1>
      Float=&FloatFromCurve(
       FloatCurve=DataInterface() {propertyValues {Curve {keys [{arriveTangent=-5.044, interpMode=RCIM_Cubic, leaveTangent=-5.044, tangentMode=RCTM_User, time=0.25, value=2}, <at 0.4 1 RCIM_Cubic>]}}}
       CurveIndex=$User.Reveal
      )
     )
     Float=&FloatFromCurve(
      FloatCurve=DataInterface() {propertyValues {Curve {keys [
          {arriveTangent=9.0725, interpMode=RCIM_Cubic, time=0, value=0}
          <at 0.5 0.9 RCIM_Cubic>
          {arriveTangent=-0.0405, interpMode=RCIM_Cubic, leaveTangent=-0.0405, tangentMode=RCTM_User, time=1, value=0}
         ]}}}
      CurveIndex=$Particles.ParticleRelativeLifetimeSprites
      Scale Curve=&Multiply_Float(A=&RandomRangeFloat(Minimum=0.5, Maximum=1), B=1)
     )
    )
    ScaleA=true
    ScaleRGB=true
   )
   &DynamicMaterialParameters(
    Index 0 Param 1=&RandomRangeFloat(Minimum=2, Maximum=1)
    Index 0 Param 2=&RandomRangeFloat(Minimum=0, Maximum=2)
    Index 0 Param 3=8
    Index 0 Param 4=$Particles.SubImageIndex
    Index 1 Param 1=&RandomRangeFloat(Minimum=-1, Maximum=1)
    Index 1 Param 2=&RandomRangeFloat(Minimum=-1, Maximum=1)
    Index 1 Param 3=&FloatFromCurve(
     FloatCurve=DataInterface() {propertyValues {Curve {keys [
         {arriveTangent=0.8752, interpMode=RCIM_Cubic, leaveTangent=0.8752, tangentMode=RCTM_User, time=0, value=0.5}
         {arriveTangent=0.9995, interpMode=RCIM_Cubic, leaveTangent=0.9995, tangentMode=RCTM_User, time=0.2223, value=0.6825}
         {arriveTangent=-0.0405, interpMode=RCIM_Cubic, leaveTangent=-0.0405, tangentMode=RCTM_User, time=0.7, value=0.9}
         <at 1 0.7>
        ]}}}
     CurveIndex=$Particles.InterpolateCurveSprites
    )
    Index 2 Param 4=1
    Module.Param0WriteEnabled=true
    Module.Param0WriteEnabled1=true
    Module.Param0WriteEnabled2=false
    Module.Param1WriteEnabled=true
    Module.Param1WriteEnabled1=true
    Module.Param1WriteEnabled2=false
    Module.Param2WriteEnabled=true
    Module.Param2WriteEnabled1=true
    Module.Param2WriteEnabled2=false
    Module.Param3WriteEnabled=true
    Module.Param3WriteEnabled1=false
    Module.Param3WriteEnabled2=true
    Write Parameter Index 1=true
    Write Parameter Index 2=true
   )
  ]
  rendererClasses [&NiagaraSpriteRendererProperties]
  renderers [{rendererClass=&NiagaraSpriteRendererProperties}]
 }
 Drips001 {
  emitterUpdateScript [
   &EmitterState(Life Cycle Mode=::NewEnumerator1, Loop Behavior=::NewEnumerator1, Loop Delay=0.03, Loop Duration=0.3, UseLoopDelay=true)
   &SpawnRate(SpawnRate=250, Use Spawn Probability=false)
  ]
  particleSpawnScript [
   &InitializeParticle(
    Color Maximum=<rgba 1 1 1 2>
    Color Minimum=<rgba 1 1 1 0.5>
    Color Mode=::NewEnumerator2
    Lifetime Max=0.15
    Lifetime Min=0.07
    Lifetime Mode=::NewEnumerator1
    Mesh Scale Max=<0.9 0.9 0.9>
    Mesh Scale Min=<0.3 0.3 0.3>
    Mesh Scale Mode=::NewEnumerator2
    Position Mode=::NewEnumerator1
    UsePositionOffset=false
    Write Material Random=true
   ) {module=InitializeParticle001}
   &ShapeLocation(
    MaskSpawnGroup=false
    Offset=<-2.5 5 0>
    Offset Coordinate Space=::Local
    Offset Mode=::NewEnumerator0
    Override Seed=false
    Sphere Radius=1
    UseOwnerScale=true
   )
   &CameraOffset()
   &AddVelocity(
    Cone Angle=30
    Cone Axis=&ReturnMeshOrientationAxis(Mesh Orientation Axis=::YAxis, Mesh Orientation Quat=$Engine.Owner.Rotation)
    Module.UseConeFalloff=false
    Module.UseVelDistribution=false
    Override Seed=false
    Rotation Coordinate Space=::World
    Velocity Mode=::NewEnumerator2
    Velocity Speed=&Multiply_Float(A=&RandomRangeFloat(Minimum=111, Maximum=444), B=1)
   )
  ]
  particleUpdateScript [
   &ParticleState()
   &DynamicMaterialParameters(
    Index 0 Param 1=&RandomRangeFloat(Minimum=0, Maximum=1)
    Index 0 Param 2=1
    Module.Param0WriteEnabled=true
    Module.Param1WriteEnabled=true
    Module.Param2WriteEnabled=true
    Module.Param3WriteEnabled=true
   )
   &Drag(Drag=&RandomRangeFloat(Minimum=2, Maximum=6), Use Linear Drag=true, Use Rotational Drag=false)
   &GravityForce(Gravity=&RandomRangeVector(Minimum=<0 0 -150>, Maximum=<0 0 -600>))
   &SolveForcesAndVelocity(Clamp Velocity=false, Limit Acceleration=false)
   &ScaleMeshSize(Scale Factor=&MakeVector(
     X=&FloatFromCurve(FloatCurve=DataInterface() {propertyValues {Curve {keys [<at 0 0.5 RCIM_Cubic>, {interpMode=RCIM_Cubic, leaveTangent=0.5, time=0.2, value=1}]}}})
     Y=&FloatFromCurve(FloatCurve=DataInterface() {propertyValues {Curve {keys [<at 0 1.5 RCIM_Cubic>, {arriveTangent=-1, interpMode=RCIM_Cubic, leaveTangent=-1, time=0.2, value=1}, <at 1 0.5>]}}})
     Z=&FloatFromCurve(FloatCurve=DataInterface() {propertyValues {Curve {keys [<at 0 1.5 RCIM_Cubic>, {arriveTangent=-1, interpMode=RCIM_Cubic, leaveTangent=-1, time=0.2, value=1}, <at 1 0.5>]}}})
    )) {module=ScaleMeshSize001}
   &ScaleColor(
    Scale Alpha=&FloatFromCurve(
     FloatCurve=DataInterface() {propertyValues {Curve {keys [
         <at 0 0 RCIM_Cubic>
         {arriveTangent=0.5321, interpMode=RCIM_Cubic, leaveTangent=0.5321, time=0.2, value=1}
         {arriveTangent=-1.4074, interpMode=RCIM_Cubic, leaveTangent=-1.4074, tangentMode=RCTM_User, time=0.5103, value=0.2715}
         {arriveTangent=-0.1766, interpMode=RCIM_Cubic, leaveTangent=-0.1766, tangentMode=RCTM_User, time=1, value=0}
        ]}}}
     Scale Curve=0.8
    )
    Scale RGB=&MakeVectorFromLinearColorRGB(Color=&ColorFromCurve(ColorCurve=DataInterface() {propertyValues {
        AlphaCurve {keys [<at 0 1 RCIM_Cubic>,<at 1 1 RCIM_Cubic>]}
        BlueCurve {keys [<at 0.0569 1 RCIM_Cubic>,<at 0.1557 0.01 RCIM_Cubic>]}
        GreenCurve {keys [<at 0.0569 1 RCIM_Cubic>,<at 0.1557 0.439 RCIM_Cubic>]}
        RedCurve {keys [<at 0.0569 1 RCIM_Cubic>,<at 0.1557 1 RCIM_Cubic>]}
       }}))
    ScaleA=true
    ScaleRGB=true
   )
  ]
  rendererClasses [&NiagaraMeshRendererProperties]
  renderers [{rendererClass=&NiagaraMeshRendererProperties, rendererIndex=0}]
 }
 Embers {
  emitterUpdateScript [
   &EmitterState()
   &SpawnRate(
    Spawn Probability=0.8
    SpawnRate=&Multiply_Float(
     A=100
     B=&Multiply_Float(
      A=&FloatFromCurve(FloatCurve=DataInterface() {propertyValues {Curve {keys [<at 0.45 1>,<at 0.55 0>]}}}, CurveIndex=$User.Reveal)
      B=1
     )
    )
    Use Spawn Probability=true
   )
  ]
  particleSpawnScript [
   &InitializeParticle(
    Color=&LerpLinearColors(
     StartColor=<rgba 10 1.0866 0 1>
     EndColor=<rgba 2 0.0333 0 1>
     LerpFactor=&UniformAOrBFloat(A=0, Distribution Weight=0.2, Evaluation Type=::NewEnumerator0, Recalculate Random Each Loop=true, Override Seed=false)
    )
    Lifetime=&RandomRangeFloat(Minimum=0.33, Maximum=1)
    Sprite Size Mode=::NewEnumerator1
    Uniform Sprite Size Max=2
    Uniform Sprite Size Min=0.8
    UsePositionOffset=false
    Write Material Random=true
   )
   &AddVelocity(
    Cone Angle=60
    Cone Axis=&ReturnMeshOrientationAxis(Mesh Orientation Axis=::YAxis, Mesh Orientation Quat=$Engine.Owner.Rotation)
    Module.UseConeFalloff=false
    Module.UseVelDistribution=false
    Override Seed=false
    Rotation Coordinate Space=::World
    Velocity Mode=::NewEnumerator2
    Velocity Speed=&Multiply_Float(A=&RandomRangeFloat(Minimum=333, Maximum=999), B=1)
   )
   &ShapeLocation(
    MaskSpawnGroup=false
    Offset=&TransformVector(Vector=<0 30 0>, Destination Space=::World)
    Offset Mode=::NewEnumerator0
    Override Seed=false
    Sphere Radius=&Multiply_Float(A=1, B=$System.FX_Scale)
    UseOwnerScale=true
   )
  ]
  particleUpdateScript [
   &ParticleState()
   &AccelerationForce(Acceleration=&Multiply_Vector(
     A=&VectorFromFloat(Value=1)
     B=&UniformRangedVector(Minimum=<0 0 150>, Maximum=<0 0 250>)
    ))
   &Drag(
    Drag=&FloatFromCurve(
     FloatCurve=DataInterface() {propertyValues {Curve {keys [<at 0 0.5>,<at 0.5 0.5>,<at 1 1>]}}}
     Scale Curve=&RandomRangeFloat(Minimum=9, Maximum=22)
    )
    Use Linear Drag=true
    Use Rotational Drag=false
   )
   /Niagara/Modules/Update/Forces/CurlNoiseForce(
    Noise Frequency:NiagaraFloat=&RandomRangeFloat(Minimum=0.05, Maximum=0.3)
    Noise Strength=&FloatFromCurve(FloatCurve=DataInterface() {propertyValues {Curve {keys [<at 0 1>,<at 1 0>]}}}, Scale Curve=500)
    Pan Noise=true
    Pan Noise Field=<0 0 1>
   )
   &ScaleSpriteSize(Scale Factor=&Vector2DFromFloat(Value=&FloatFromCurve(FloatCurve=DataInterface() {propertyValues {Curve {keys [
          {arriveTangent=9.1531, interpMode=RCIM_Cubic, leaveTangent=9.1531, tangentMode=RCTM_User, time=0, value=0.02}
          {arriveTangent=-0.02, interpMode=RCIM_Cubic, leaveTangent=-0.02, time=0.18, value=0.99}
          <at 1 0 RCIM_Cubic>
         ]}}}))) {module=ScaleSpriteSize002}
   &GravityForce(Gravity=<0 0 -600>)
   &SolveForcesAndVelocity(Clamp Velocity=false, Limit Acceleration=false)
   &ScaleColor(
    Scale Alpha=&FloatFromCurve(
     FloatCurve=DataInterface() {propertyValues {Curve {keys [
         {arriveTangent=9.1531, interpMode=RCIM_Cubic, leaveTangent=9.1531, tangentMode=RCTM_User, time=0, value=0.02}
         {arriveTangent=0.48, interpMode=RCIM_Cubic, leaveTangent=0.48, time=0.18, value=0.99}
         <at 1 0.5 RCIM_Cubic>
        ]}}}
     Scale Curve=&MakeCustomFloatFromBool(Bool=$User.bIsFrontEndPreview, True Float=0.5, False Float=1)
    )
    Scale RGB=&MakeVectorFromLinearColorRGB(Color=&ColorFromCurve(ColorCurve=DataInterface() {propertyValues {
        AlphaCurve {keys [<at 0 1 RCIM_Cubic>]}
        BlueCurve {keys [<at 0 1.2255 RCIM_Cubic>,<at 0.5978 0 RCIM_Cubic>]}
        GreenCurve {keys [<at 0 3.7041 RCIM_Cubic>,<at 0.5978 0.0133 RCIM_Cubic>]}
        RedCurve {keys [<at 0 5 RCIM_Cubic>,<at 0.5978 0.8 RCIM_Cubic>]}
       }}))
    ScaleA=true
    ScaleRGB=true
   ) {module=ScaleColor001}
   &ScaleSpriteSizeBySpeed(
    Max Scale Factor=<1.5 5>
    Min Scale Factor=<0.4 0.8>
    Sample Scale Factor By Curve=true
    Scale Factor Curve=DataInterface() {propertyValues {Curve {keys [
        <at -0.0077 0.0765 RCIM_Cubic>
        {arriveTangent=1.2463, interpMode=RCIM_Cubic, leaveTangent=1.2463, tangentMode=RCTM_User, time=0.9985, value=0.9549}
       ]}}}
    Velocity Threshold=888
   )
   &FN_Sprite_Attributes(
    Color=&MultiplyLinearColor(Linear Color=$Particles.Color, Scale Factor=<1 1 1 0.005>)
    Create Color Attribute=true
    Create SpriteSize Attribute=true
    SpriteSize=&Multiply_Vector2D_ByFloat(Float=5, Vector2D=$Particles.SpriteSize)
    Write to Module Namespace=true
   ) {module=FN_Sprite_Attributes001}
  ]
  rendererClasses [&NiagaraSpriteRendererProperties]
  renderers [
   {rendererClass=&NiagaraSpriteRendererProperties}
   {rendererClass=&NiagaraSpriteRendererProperties, rendererIndex=1}
  ]
 }
 FuelSpit {
  emitterUpdateScript [
   &EmitterState(Life Cycle Mode=::NewEnumerator1, Loop Behavior=::NewEnumerator1, Loop Duration=0.3, UseLoopDelay=false)
   &SpawnRate(SpawnRate=10, Use Spawn Probability=false)
  ]
  particleSpawnScript [
   &InitializeParticle(
    Color=<rgba 0.679 0.6945 0.7 0.1333>
    Lifetime Max=0.15
    Lifetime Min=0.07
    Lifetime Mode=::NewEnumerator1
    Mesh Scale=&Multiply_VectorByFloat(
     Vector=<1 0.222 1>
     Float=&Multiply_Float(A=&RandomRangeFloat(Minimum=0.2, Maximum=0.3), B=1)
    )
    Mesh Scale Mode=::NewEnumerator4
    Position Offset=<10 0 0>
    Sprite Rotation Mode=::NewEnumerator2
    UsePositionOffset=true
    Write Material Random=true
   )
   &AddVelocity(
    Rotation Coordinate Space=::World
    Velocity=&ReturnMeshOrientationAxis(Mesh Orientation Axis=::YAxis, Mesh Orientation Quat=$Engine.Owner.Rotation)
   ) {module=AddVelocity001}
   &SystemLocation(MaskSpawnGroup=false, Offset=<-3 20 0>)
  ]
  particleUpdateScript [
   &Drag(Drag=1.6, Use Linear Drag=true, Use Rotational Drag=true)
   &SolveForcesAndVelocity(Clamp Velocity=false, Limit Acceleration=false)
   &ParticleState()
   &ScaleMeshSize(Scale Factor=&MakeVector(
     X=&FloatFromCurve(FloatCurve=DataInterface() {propertyValues {Curve {keys [
          {interpMode=RCIM_Cubic, tangentMode=RCTM_User, time=0, value=1}
          {arriveTangent=2.0132, interpMode=RCIM_Cubic, leaveTangent=2.0132, tangentMode=RCTM_User, time=1, value=2}
         ]}}})
     Y=&FloatFromCurve(FloatCurve=DataInterface() {propertyValues {Curve {keys [
          {arriveTangent=-0.0067, interpMode=RCIM_Cubic, leaveTangent=-0.0067, tangentMode=RCTM_User, time=0.03, value=0.1}
          {arriveTangent=0.6408, interpMode=RCIM_Cubic, leaveTangent=0.6408, tangentMode=RCTM_User, time=0.3, value=1}
         ]}}})
     Z=&FloatFromCurve(FloatCurve=DataInterface() {propertyValues {Curve {keys [
          {arriveTangent=-0.0067, interpMode=RCIM_Cubic, leaveTangent=-0.0067, tangentMode=RCTM_User, time=0.03, value=0.1}
          {arriveTangent=0.6408, interpMode=RCIM_Cubic, leaveTangent=0.6408, tangentMode=RCTM_User, time=0.3, value=1}
         ]}}})
    ))
   &ScaleColor(
    Scale Alpha=&FloatFromCurve(
     FloatCurve=DataInterface() {propertyValues {Curve {keys [
         {arriveTangent=0.1658, interpMode=RCIM_Cubic, leaveTangent=0.1658, tangentMode=RCTM_User, time=0, value=0}
         {arriveTangent=-0.3124, interpMode=RCIM_Cubic, leaveTangent=-0.3124, tangentMode=RCTM_User, time=0.333, value=1}
         {arriveTangent=0.0367, interpMode=RCIM_Cubic, leaveTangent=0.0367, tangentMode=RCTM_User, time=1, value=0}
        ]}}}
     Scale Curve=2
    )
    Scale RGB=&MakeVectorFromLinearColorRGB(Color=&ColorFromCurve(ColorCurve=DataInterface() {propertyValues {
        AlphaCurve {keys [<at 0 1 RCIM_Cubic>,<at 1 1 RCIM_Cubic>]}
        BlueCurve {keys [<at 0.2761 1 RCIM_Cubic>,<at 0.4423 0.02 RCIM_Cubic>]}
        GreenCurve {keys [<at 0.2761 1 RCIM_Cubic>,<at 0.4423 0.878 RCIM_Cubic>]}
        RedCurve {keys [<at 0.2761 1 RCIM_Cubic>,<at 0.4423 2 RCIM_Cubic>]}
       }}))
    ScaleA=true
    ScaleRGB=true
   )
   &DynamicMaterialParameters(
    Index 0 Param 1=&FloatFromCurve(FloatCurve=DataInterface() {propertyValues {Curve {keys [<at 0 0>,<at 1 1>]}}})
    Module.Index0Param0Write=true
    Module.Index0Param1Write=false
    Module.Index0Param2Write=false
    Module.Index0Param3Write=false
   )
  ]
  rendererClasses [&NiagaraMeshRendererProperties]
  renderers [{rendererClass=&NiagaraMeshRendererProperties, rendererIndex=0}]
 }
 InitialFlameBurst {
  emitterUpdateScript [
   &EmitterState()
   &SpawnBurst_Instantaneous(
    Spawn Count=&RandomRangeInt(Minimum=12, Maximum=16)
    Spawn Time=0.03
    Use Spawn Probability=false
    UseLoopCountLimit=false
   )
  ]
  particleSpawnScript [
   &InitializeParticle(
    Color=<rgba 1 1 1 0.6>
    Lifetime=&UniformRangedFloat(Minimum=0.25, Maximum=0.4)
    Mass Mode=::NewEnumerator1
    Sprite Rotation Angle=&UniformRangedFloat(Minimum=-40, Maximum=40)
    Sprite Rotation Mode=::NewEnumerator0
    Sprite Size=&Vector2DFromFloat(Value=&UniformRangedFloat(Minimum=25, Maximum=40))
    Sprite Size Mode=::NewEnumerator4
    UsePositionOffset=false
    Write Material Random=true
   ) {module=InitializeParticle001}
   &ShapeLocation(
    MaskSpawnGroup=false
    Offset=<-8 20 0>
    Offset Coordinate Space=::Local
    Offset Mode=::NewEnumerator0
    Override Seed=false
    Sphere Radius=1
    UseOwnerScale=true
   ) {module=ShapeLocation001}
   &AddVelocity(
    Cone Angle=66
    Cone Axis=$Engine.Owner.SystemYAxis
    Module.UseConeFalloff=false
    Module.UseVelDistribution=false
    Override Seed=false
    Rotation Coordinate Space=::World
    Velocity Mode=::NewEnumerator2
    Velocity Speed=&Multiply_Float(A=&RandomRangeFloat(Minimum=40, Maximum=200), B=1)
   ) {module=AddVelocity001}
  ]
  particleUpdateScript [
   &ParticleState()
   &SubUVAnimation(SubUV Lookup Index:NiagaraFloat=&Multiply_Float(A=$Particles.NormalizedAge, B=0.5))
   &InheritSourceMovement(ApplyForce=false, ApplyPosition=true, ApplyPositionFalloff=false)
   &SolveForcesAndVelocity(Clamp Velocity=false, Limit Acceleration=false)
   &ScaleColor(
    Scale Alpha=&Multiply_Float(
     A=&FloatFromCurve(FloatCurve=DataInterface() {propertyValues {Curve {keys [
          {arriveTangent=16.5612, interpMode=RCIM_Cubic, leaveTangent=16.5612, leaveTangentWeight=0.3333, tangentMode=RCTM_User, time=0, value=0}
          {arriveTangentWeight=0.0333, interpMode=RCIM_Cubic, leaveTangentWeight=0.3, tangentMode=RCTM_User, time=0.1, value=1}
          {arriveTangentWeight=0.1333, interpMode=RCIM_Cubic, leaveTangentWeight=0.1667, tangentMode=RCTM_User, time=0.5, value=1}
          {arriveTangentWeight=0.1667, interpMode=RCIM_Cubic, tangentMode=RCTM_User, time=1, value=0}
         ]}}})
     B=&RandomRangeFloat(Minimum=0.5, Maximum=1)
    )
    Scale RGB=&Multiply_VectorByFloat(
     Vector=&Multiply_VectorByFloat(
      Vector=<1 1 1>
      Float=&FloatFromCurve(
       FloatCurve=DataInterface() {propertyValues {Curve {keys [
           {arriveTangent=-5.355, interpMode=RCIM_Cubic, leaveTangent=-5.355, tangentMode=RCTM_User, time=0.25, value=1.5}
           {interpMode=RCIM_Cubic, leaveTangent=2.8296, time=0.45, value=1}
          ]}}}
       CurveIndex=$User.Reveal
      )
     )
     Float=&FloatFromCurve(
      FloatCurve=DataInterface() {propertyValues {Curve {keys [
          {arriveTangent=9.0725, interpMode=RCIM_Cubic, leaveTangent=9.0725, tangentMode=RCTM_User, time=0, value=0}
          {arriveTangent=-2.5249, interpMode=RCIM_Cubic, leaveTangent=-2.5249, tangentMode=RCTM_User, time=0.1408, value=2}
          {arriveTangent=-0.0405, interpMode=RCIM_Cubic, leaveTangent=-0.0405, tangentMode=RCTM_User, time=1, value=0}
         ]}}}
      Scale Curve=&RandomRangeFloat(Minimum=0.9, Maximum=1.5)
     )
    )
    ScaleA=true
    ScaleRGB=true
   )
   &ScaleSpriteSize(Scale Factor=&Vector2DFromFloat(Value=&FloatFromCurve(FloatCurve=DataInterface() {propertyValues {Curve {keys [
          <at 0 0.7 RCIM_Cubic>
          {arriveTangent=-0.565, interpMode=RCIM_Cubic, leaveTangent=-0.565, tangentMode=RCTM_User, time=0.2, value=1.5}
          {interpMode=RCIM_Cubic, tangentMode=RCTM_User, time=1, value=0.75}
         ]}}}))) {module=ScaleSpriteSize001}
   &DynamicMaterialParameters(
    Index 0 Param 1=&RandomRangeFloat(Minimum=2, Maximum=1)
    Index 0 Param 2=&RandomRangeFloat(Minimum=0, Maximum=2)
    Index 0 Param 3=8
    Index 0 Param 4=$Particles.SubImageIndex
    Index 1 Param 1=&RandomRangeFloat(Minimum=-1, Maximum=1)
    Index 1 Param 2=&RandomRangeFloat(Minimum=-1, Maximum=1)
    Index 1 Param 3=&FloatFromCurve(FloatCurve=DataInterface() {propertyValues {Curve {keys [
         {arriveTangent=4.1785, interpMode=RCIM_Cubic, leaveTangent=4.1785, tangentMode=RCTM_User, time=0, value=0}
         {arriveTangent=0.9995, interpMode=RCIM_Cubic, leaveTangent=0.9995, tangentMode=RCTM_User, time=0.2223, value=0.6825}
         {arriveTangent=-0.0405, interpMode=RCIM_Cubic, leaveTangent=-0.0405, tangentMode=RCTM_User, time=0.7, value=0.9}
         <at 1 0.7>
        ]}}})
    Index 2 Param 4=1
    Module.Param0WriteEnabled=true
    Module.Param0WriteEnabled1=true
    Module.Param0WriteEnabled2=false
    Module.Param1WriteEnabled=true
    Module.Param1WriteEnabled1=true
    Module.Param1WriteEnabled2=false
    Module.Param2WriteEnabled=true
    Module.Param2WriteEnabled1=true
    Module.Param2WriteEnabled2=false
    Module.Param3WriteEnabled=true
    Module.Param3WriteEnabled1=false
    Module.Param3WriteEnabled2=true
    Write Parameter Index 1=true
    Write Parameter Index 2=true
   )
  ]
  rendererClasses [&NiagaraSpriteRendererProperties]
  renderers [{rendererClass=&NiagaraSpriteRendererProperties}]
 }
 InitialGlow {
  emitterUpdateScript [
   &EmitterState()
   &SpawnBurst_Instantaneous(Spawn Time=0.005, Use Spawn Probability=false, UseLoopCountLimit=false)
  ]
  particleSpawnScript [
   &InitializeParticle(
    Color=<rgba 20 5.3 0.4 1>
    Lifetime=0.45
    Sprite Rotation Mode=::NewEnumerator2
    Sprite Size=<100 150>
    Sprite Size Mode=::NewEnumerator4
    UsePositionOffset=false
    Write Material Random=true
   )
   &ShapeLocation(
    MaskSpawnGroup=false
    Offset=<0 25 0>
    Offset Coordinate Space=::Local
    Offset Mode=::NewEnumerator0
    Override Seed=false
    Sphere Radius=1
    UseOwnerScale=true
   ) {module=ShapeLocation001}
   &AddVelocity(
    Cone Angle=1
    Cone Axis=$Engine.Owner.SystemYAxis
    Module.UseConeFalloff=false
    Module.UseVelDistribution=false
    Override Seed=false
    Rotation Coordinate Space=::World
    Velocity Mode=::NewEnumerator2
    Velocity Speed=1
   ) {module=AddVelocity001}
   &ApplyInitialForces()
   &CameraOffset(Camera Offset Amount=20)
  ]
  particleUpdateScript [
   /SetVariables(Particles.ColorGlow:LinearColor=&MultiplyLinearColor(
     Linear Color=&ColorFromCurve(ColorCurve=DataInterface() {propertyValues {
        AlphaCurve {keys [<at 0 0 RCIM_Cubic>,<at 0.1223 0.15>,<at 0.7125 0 RCIM_Cubic>]}
        BlueCurve {keys [<at 0.1587 0.5 RCIM_Cubic>,<at 0.5812 0.06>]}
        GreenCurve {keys [<at 0.1587 2.375 RCIM_Cubic>,<at 0.5812 0.501>]}
        RedCurve {keys [<at 0.1587 5 RCIM_Cubic>,<at 0.5812 3>]}
       }})
     Scale Factor=&MakeFloat4FromVectorAndFloat(
      "Vector (RGB)"=<1 1 1>
      "Float (Alpha)"=&Multiply_Float(A=$System.ToD, B=&FN_FrontEndFloatModifier(Gameplay=1, Front End=0.3))
     )
    )) {isSetParameters=true}
   &ParticleState()
   &ScaleColor(
    Scale Alpha=&Waveform(
     "Waveform [1]"=::NewEnumerator0
     "[1] Frequency"=1
     "[1] Amplitude Min/Max"=<0.5 1>
     "[1] Scale"=&Waveform(
      Waveform Count=::NewEnumerator1
      Global Input Position=0.33
      "Waveform [1]"=::NewEnumerator0
      "[1] Frequency"=0.33
      "[1] Amplitude Min/Max"=<0.5 1.5>
      "[2] Average"=false
      "[2] Input Position"=0
      "[2] Phase"=$Emitter.Age
      "[2] Phase Scale"=1
      "[2] Compound Frequency"=2
      "[2] Frequency"=2
      "[2] Amplitude Min/Max"=<0 1>
      "[2] Scale"=1
      "[2] Bias"=0
      "[2] Exponent"=1
      "[2] Pulse Duration"=0.1
      "[2] Clip Triangle Min/Max"=<0 0>
      "[2] Random Seed"=0
     )
    )
    ScaleA=true
    ScaleRGB=true
   )
   &ScaleColor(
    Scale Alpha=&FloatFromCurve(
     FloatCurve=DataInterface() {propertyValues {Curve {keys [{interpMode=RCIM_Cubic, tangentMode=RCTM_User, time=0.4, value=1}, {interpMode=RCIM_Cubic, tangentMode=RCTM_User, time=1, value=0}]}}}
     Scale Curve=$System.ToD
    )
    ScaleA=true
    ScaleRGB=true
   ) {module=ScaleColor001}
   /SetVariables(Particles.Position=&MakeVector(X=0, Y=&Lerp_Float(A=22, B=70, Alpha=$Particles.NormalizedAge), Z=0)) {isSetParameters=true}
  ]
  rendererClasses [&NiagaraSpriteRendererProperties]
  renderers [{rendererClass=&NiagaraSpriteRendererProperties}]
 }
 Light {
  emitterUpdateScript [
   &EmitterState()
   &SpawnBurst_Instantaneous(Use Spawn Probability=false, UseLoopCountLimit=false)
  ]
  particleSpawnScript [&InitializeParticle(UsePositionOffset=false, Write Material Random=true)]
  particleUpdateScript [
   &ParticleState(Kill Particles When Lifetime Has Elapsed=false)
   /SetVariables(Particles.Position=&AddVectorToPosition(
     Position=&SimulationPosition()
     Vector=&Multiply_VectorByFloat(
      Vector=&RotateVectorByQuaternion(
       Vector To Rotate=<0 1 0>
       Quaternion=&LerpQuaternion(
        Quaternion A:Quat4f=$Engine.Owner.Rotation
        Quaternion B:Quat4f=$System.InterpolateOverTime.InterpolateOverTime_Quaternion.Moving Average
        Lerp Factor:NiagaraFloat=0.5
        Enable Spherical Linear Interpolation:NiagaraBool=false
       )
      )
      Float=88
     )
    )) {isSetParameters=true}
   &ScaleColor(
    Scale Alpha=&FloatFromCurve(
     FloatCurve=DataInterface() {propertyValues {Curve {keys [
         {interpMode=RCIM_Cubic, tangentMode=RCTM_User, time=0, value=0}
         {interpMode=RCIM_Cubic, tangentMode=RCTM_User, time=0.5, value=1}
         {tangentMode=RCTM_User, time=1, value=0}
        ]}}}
     CurveIndex=$User.Reveal
    )
    ScaleA=true
    ScaleRGB=true
   )
   &FN_Light_Attributes(
    Color=&ColorFromCurve(
     ColorCurve=DataInterface() {propertyValues {
       AlphaCurve {keys [<at 0 0>,<at 0.166 1 RCIM_Cubic>,<at 0.75 0 RCIM_Cubic>]}
       BlueCurve {keys [<at 0.1573 0 RCIM_Cubic>,<at 0.3769 0 RCIM_Cubic>,<at 1 0>]}
       GreenCurve {keys [<at 0.1573 0.95 RCIM_Cubic>, {arriveTangent=-0.9098, interpMode=RCIM_Cubic, leaveTangent=-0.9098, time=0.3769, value=0.25}, <at 1 0.1833>]}
       RedCurve {keys [<at 0.1573 3 RCIM_Cubic>, {arriveTangent=-2.3733, interpMode=RCIM_Cubic, leaveTangent=-2.3733, time=0.3769, value=1}, <at 1 1>]}
      }}
     CurveIndex=$User.Reveal
    )
    Create Color Attribute=true
    Create Exponent Attribute=true
    Create Radius Attribute=true
    Exponent=3
    Radius=&FloatFromCurve(
     FloatCurve=DataInterface() {propertyValues {Curve {keys [<at 0 0.5>,<at 0.1 1>,<at 0.75 0.85 RCIM_Constant>,<at 0.766 0>]}}}
     Scale Curve=166
    )
    Scale Color=&Multiply_Float(A=20, B=&FN_FrontEndFloatModifier(Gameplay=1, Front End=0.4))
   )
  ]
  rendererClasses [&NiagaraLightRendererProperties]
  renderers [{rendererClass=&NiagaraLightRendererProperties, rendererIndex=0}]
 }
 MainBurst003 {
  emitterSpawnScript [/SetVariables(Emitter.Length=166, Emitter.ParticlesToSpawn=32) {isSetParameters=true, module=SetVariables_#3}]
  emitterUpdateScript [
   &EmitterState(Life Cycle Mode=::NewEnumerator1, Loop Behavior=::NewEnumerator1, Loop Duration=1, UseLoopDelay=false)
   &SpawnBurst_Instantaneous(Spawn Count=$Emitter.ParticlesToSpawn, Use Spawn Probability=false, UseLoopCountLimit=false)
   /SetVariables(Emitter.Reveal:NiagaraFloat=$User.Reveal) {isSetParameters=true, module=SetVariables_#4}
   /SetVariables(Emitter.Length:NiagaraFloat=&Multiply_Float(
     A=$Emitter.Initial.Length
     B=&FloatFromCurve(
      FloatCurve=DataInterface() {propertyValues {Curve {keys [{interpMode=RCIM_Cubic, leaveTangent=2.4356, tangentMode=RCTM_Break, time=0, value=0}, <at 1 1>]}}}
      CurveIndex=$Emitter.Reveal
     )
    )) {isSetParameters=true, module=SetVariables_#5}
  ]
  particleSpawnScript [&InitializeParticle(Color Mode=::NewEnumerator0, Sprite Rotation Mode=::NewEnumerator2, UsePositionOffset=false, Write Material Random=true)]
  particleUpdateScript [
   &ParticleState(Kill Particles When Lifetime Has Elapsed=false)
   &SpriteRotationRate(Rotation Rate=&RandomRangeFloat(Minimum=-444, Maximum=444, Evaluation Type=::NewEnumerator1))
   /SetVariables(Particles.ParticleRelativeLifetimeRibbons:NiagaraFloat=&ReturnNormalizedExecIndex()) {isSetParameters=true, module=SetVariables_#6}
   /SetVariables(Particles.OpacityMultiplierRibbons:NiagaraFloat=&FloatFromCurve(
     FloatCurve=DataInterface() {propertyValues {Curve {keys [
         {interpMode=RCIM_Cubic, tangentMode=RCTM_User, time=0, value=0}
         {interpMode=RCIM_Cubic, tangentMode=RCTM_User, time=0.5, value=1}
         {tangentMode=RCTM_User, time=0.9, value=0}
        ]}}}
     CurveIndex=&Add_Float(
      A=&Multiply_Float(A=$Particles.ParticleRelativeLifetimeRibbons, B=-1)
      B=&Multiply_Float(A=$Emitter.Reveal, B=2)
     )
    )) {isSetParameters=true, module=SetVariables_#7}
   /SetVariables(Particles.InterpolateCurveRibbons:NiagaraFloat=&FloatFromCurve(
     FloatCurve=DataInterface() {propertyValues {Curve {keys [
         {arriveTangent=-0.6503, interpMode=RCIM_Cubic, leaveTangent=1.1852, tangentMode=RCTM_Break, time=0, value=0}
         {arriveTangent=0.8679, interpMode=RCIM_Cubic, tangentMode=RCTM_Break, time=1, value=1}
        ]}}}
     CurveIndex=$Particles.ParticleRelativeLifetimeRibbons
    )) {isSetParameters=true, module=SetVariables_#8}
   /SetVariables(Particles.PositionRibbons:NiagaraPosition=&AddVectorToPosition(
     Position=&AddVectorToPosition(
      Position=&SimulationPosition()
      Vector=&TransformVector(Vector=<-3 10 0>, Destination Space=::Simulation)
     )
     Vector=&Multiply_VectorByFloat(
      Vector=&RotateVectorByQuaternion(
       Vector To Rotate=<0 1 0>
       Quaternion=&LerpQuaternion(
        Quaternion A:Quat4f=$Engine.Owner.Rotation
        Quaternion B:Quat4f=$System.InterpolateOverTime.InterpolateOverTime_Quaternion.Moving Average
        Lerp Factor:NiagaraFloat=$Particles.InterpolateCurveRibbons
        Enable Spherical Linear Interpolation:NiagaraBool=false
       )
      )
      Float=&Multiply_Float(A=$Emitter.Length, B=$Particles.InterpolateCurveRibbons)
     )
    )) {isSetParameters=true, module=SetVariables_#9}
   /SetVariables(
    Particles.BackRibbonWidth:NiagaraFloat=&FloatFromCurve(
     FloatCurve=DataInterface() {propertyValues {Curve {keys [
         {arriveTangent=1.636, interpMode=RCIM_Cubic, leaveTangent=1.636, tangentMode=RCTM_User, time=0, value=0.0777}
         {arriveTangent=1.0523, interpMode=RCIM_Cubic, leaveTangent=1.0523, tangentMode=RCTM_User, time=0.2799, value=0.5408}
         {arriveTangent=-0.0372, interpMode=RCIM_Cubic, leaveTangent=-0.0372, tangentMode=RCTM_User, time=0.7, value=0.8}
         {arriveTangent=-1.5957, interpMode=RCIM_Cubic, leaveTangent=-1.5957, tangentMode=RCTM_User, time=1, value=0.5138}
        ]}}}
     CurveIndex=$Particles.InterpolateCurveRibbons
     Scale Curve=25
    )
    Particles.ColorRibbonBack:LinearColor=&ColorFromCurve(
     ColorCurve=DataInterface() {propertyValues {
       AlphaCurve {keys [
         <at 0 0>
         <at 0.2133 1>
         {arriveTangent=-1.2718, interpMode=RCIM_Cubic, leaveTangent=-1.2718, time=0.7268, value=1}
         <at 0.9995 0 RCIM_Cubic>
        ]}
       BlueCurve {keys [<at 0.4438 0>]}
       GreenCurve {keys [<at 0.4438 0.016>]}
       RedCurve {keys [<at 0.4438 0.08>]}
      }}
     CurveIndex=$Particles.ParticleRelativeLifetimeRibbons
    )
    Particles.ColorRibbonFill:LinearColor=&MultiplyLinearColor(
     Linear Color=&ColorFromCurve(
      ColorCurve=DataInterface() {propertyValues {
        AlphaCurve {keys [
          <at 0 0>
          <at 0.2133 0.8>
          {arriveTangent=-1.0175, interpMode=RCIM_Cubic, leaveTangent=-1.0175, time=0.7268, value=0.8}
          <at 0.9995 0 RCIM_Cubic>
         ]}
        BlueCurve {keys [<at 0.4438 0>]}
        GreenCurve {keys [<at 0.4438 0.0933>]}
        RedCurve {keys [<at 0.4438 0.7>]}
       }}
      CurveIndex=$Particles.ParticleRelativeLifetimeRibbons
     )
     Scale Factor=&MakeFloat4FromVectorAndFloat(
      "Vector (RGB)"=&VectorFromFloat(Value=&FloatFromCurve(
        FloatCurve=DataInterface() {propertyValues {Curve {keys [
            {arriveTangent=-3964.5342, interpMode=RCIM_Cubic, leaveTangent=-3964.5342, tangentMode=RCTM_User, time=0.25, value=400}
            <at 0.45 1 RCIM_Cubic>
           ]}}}
        CurveIndex=$User.Reveal
       ))
      "Float (Alpha)"=1
     )
    )
    Particles.ColorRibbonOrange:LinearColor=&MultiplyLinearColor(
     Linear Color=&ColorFromCurve(
      ColorCurve=DataInterface() {propertyValues {
        AlphaCurve {keys [
          <at 0.0164 0>
          <at 0.2133 0.8>
          {arriveTangent=-1.0175, interpMode=RCIM_Cubic, leaveTangent=-1.0175, time=0.7268, value=0.8}
          <at 0.9995 0 RCIM_Cubic>
         ]}
        BlueCurve {keys [<at 0.4438 0.03>]}
        GreenCurve {keys [<at 0.4438 0.2505>]}
        RedCurve {keys [<at 0.4438 1.5>]}
       }}
      CurveIndex=$Particles.ParticleRelativeLifetimeRibbons
     )
     Scale Factor=&MakeFloat4FromVectorAndFloat(
      "Vector (RGB)"=&VectorFromFloat(Value=&FloatFromCurve(
        FloatCurve=DataInterface() {propertyValues {Curve {keys [
            {arriveTangent=-3964.5342, interpMode=RCIM_Cubic, leaveTangent=-3964.5342, tangentMode=RCTM_User, time=0.25, value=400}
            <at 0.45 1 RCIM_Cubic>
           ]}}}
        CurveIndex=$User.Reveal
       ))
      "Float (Alpha)"=1
     )
    )
    Particles.RibbonWidth=&FloatFromCurve(
     FloatCurve=DataInterface() {propertyValues {Curve {keys [
         {arriveTangent=0.5131, interpMode=RCIM_Cubic, time=0, value=0.066}
         {arriveTangent=1.2515, interpMode=RCIM_Cubic, leaveTangent=1.2515, time=0.351, value=0.3349}
         {arriveTangent=-0.003, interpMode=RCIM_Cubic, leaveTangent=-0.003, time=0.75, value=1.0046}
         {arriveTangent=-3.6992, interpMode=RCIM_Cubic, leaveTangent=-3.6992, tangentMode=RCTM_User, time=1, value=0.333}
        ]}}}
     CurveIndex=$Particles.InterpolateCurveRibbons
     Scale Curve=44
    )
   ) {isSetParameters=true, module=SetVariables_#10}
   /SetVariables(
    Particles.RibbonWidthBack:NiagaraFloat=&Multiply_Float(A=$Particles.BackRibbonWidth, B=3)
    Particles.RibbonWidthFill:NiagaraFloat=&Multiply_Float(A=$Particles.BackRibbonWidth, B=2.5)
    Particles.RibbonWidthOrange:NiagaraFloat=&Multiply_Float(A=$Particles.BackRibbonWidth, B=2.3)
   ) {isSetParameters=true}
   /SetVariables(
    Particles.ScaledColorRibbonBack:LinearColor=&MultiplyLinearColor(
     Linear Color=$Particles.ColorRibbonBack
     Scale Factor=&MakeFloat4FromVectorAndFloat("Vector (RGB)"=<1 1 1>, "Float (Alpha)"=&Multiply_Float(A=$Particles.OpacityMultiplierRibbons, B=0.7))
    )
    Particles.ScaledColorRibbonFill:LinearColor=&MultiplyLinearColor(
     Linear Color=$Particles.ColorRibbonFill
     Scale Factor=&MakeFloat4FromVectorAndFloat("Vector (RGB)"=<1 1 1>, "Float (Alpha)"=&Multiply_Float(A=$Particles.OpacityMultiplierRibbons, B=0.7))
    )
    Particles.ScaledColorRibbonOrange:LinearColor=&MultiplyLinearColor(
     Linear Color=$Particles.ColorRibbonOrange
     Scale Factor=&MakeFloat4FromVectorAndFloat("Vector (RGB)"=<1 1 1>, "Float (Alpha)"=&Multiply_Float(A=$Particles.OpacityMultiplierRibbons, B=0.5))
    )
   ) {isSetParameters=true}
  ]
  rendererClasses [&NiagaraRibbonRendererProperties]
  renderers [
   {rendererClass=&NiagaraRibbonRendererProperties}
   {rendererClass=&NiagaraRibbonRendererProperties, rendererIndex=1}
   {rendererClass=&NiagaraRibbonRendererProperties, rendererIndex=2}
  ]
 }
}
summary {
 systemName=NS_Emote_StrategicSpur_FireSpit
 emitters [
  {emitterName=Light, rendererClasses [&NiagaraLightRendererProperties]}
  {emitterName=Embers, rendererClasses [&NiagaraSpriteRendererProperties]}
  {emitterName=InitialFlameBurst, rendererClasses [&NiagaraSpriteRendererProperties]}
  {emitterName=FuelSpit, rendererClasses [&NiagaraMeshRendererProperties]}
  {emitterName=Drips001, rendererClasses [&NiagaraMeshRendererProperties]}
  {emitterName=MainBurst003, rendererClasses [&NiagaraRibbonRendererProperties]}
  {emitterName=AdditionalFlames001, rendererClasses [&NiagaraSpriteRendererProperties]}
  {emitterName=InitialGlow, rendererClasses [&NiagaraSpriteRendererProperties]}
 ]
 userVariables [
  User.bIsFrontEnd {type=NiagaraBool}
  User.bIsFrontEndPreview {type=NiagaraBool}
  User.bIsLargeBodyPawn {type=NiagaraBool}
  User.Reveal {type=NiagaraFloat, defaultValue=0.5, description=""}
 ]
}
