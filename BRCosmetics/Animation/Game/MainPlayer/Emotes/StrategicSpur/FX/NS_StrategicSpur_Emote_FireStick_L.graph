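# Niagara system for the left fire stick in the StrategicSpur emote, drawing a burning flame that trails as it moves.
# A mesh core, two flame sprite layers and a pulsing glow with particle light form the fire.
# Embers spawn by rate and distance moved with curl noise, while an anim-trail ribbon follows skeletal mesh bones or sockets.
# System fade-out, time-of-day and front-end modifiers adjust parts of the brightness and light.
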NiagaraSystem /BRCosmetics/Animation/Game/MainPlayer/Emotes/StrategicSpur/FX/NS_StrategicSpur_Emote_FireStick_L
dependencies {
 usedDataInterfaces [&NiagaraDataInterfaceCurve, &NiagaraDataInterfaceColorCurve, &NiagaraDataInterfaceSkeletalMesh]
 usedDynamicInputs []
 usedModules [
  &self:SystemSpawnScript.SystemScriptSource.SystemScriptGraph.NiagaraNodeAssignment_0.SetVariables_#1
  &SystemState
  &self:SystemSpawnScript.SystemScriptSource.SystemScriptGraph.NiagaraNodeAssignment_1.SetVariables_#2
  &self:SystemSpawnScript.SystemScriptSource.SystemScriptGraph.NiagaraNodeAssignment_2.SetVariables_#3
 ]
}
emitters {
 FN_Debug {
  bEnabled=false
  emitterUpdateScript [
   &EmitterState()
   &SpawnBurst_Instantaneous(Spawn Count=1, Use Spawn Probability=false)
  ]
  particleSpawnScript [/Niagara/Modules/Spawn/Initialization/InitializeParticle(
    Lifetime=5
    Module.Write Color=true
    Module.Write Lifetime=true
    Module.Write Mass=true
    Module.Write Material Random=true
    Module.Write Position=false
    Module.Write Scale=false
    Module.Write SpriteRotation=false
    Module.Write SpriteSize=true
    Position=&SimulationPosition()
    Sprite Size=<100 100>
   )]
  particleUpdateScript [
   &ParticleState()
   /SetVariables(Particles.Debug:NiagaraFloat=&FN_FrontEndFloatModifier(Gameplay=0, Front End=420)) {isSetParameters=true}
   &DynamicMaterialParameters(
    Index 0 Param 1=$Particles.Debug
    Module.Param0WriteEnabled=true
    Module.Param1WriteEnabled=true
    Module.Param2WriteEnabled=true
    Module.Param3WriteEnabled=true
   )
  ]
  rendererClasses [&NiagaraSpriteRendererProperties]
  renderers [{rendererClass=&NiagaraSpriteRendererProperties}]
 }
 FireLicks_Sprites_Tall001 {
  emitterUpdateScript [
   &EmitterState(Enable Distance Culling=true, MaxDistance=3000, Module.bUseMaxDistance=true, Module.bUseMinDistance=false, Scalability Mode=::NewEnumerator1)
   &SpawnRate(SpawnRate=8, Use Spawn Probability=false)
  ]
  particleSpawnScript [
   /Niagara/Modules/Spawn/Initialization/V2/InitializeParticle(
    Color=<rgba 1 1 1 0.333>
    Lifetime Max=1
    Lifetime Min=0.5
    Lifetime Mode=::NewEnumerator1
    Mass Mode=::NewEnumerator1
    Sprite Rotation Mode=::NewEnumerator2
    Sprite Size Mode=::NewEnumerator1
    Uniform Sprite Size Max=22
    Uniform Sprite Size Min=15
    UsePositionOffset=false
    Write Material Random=true
   ) {module=InitializeParticle001}
   &DynamicMaterialParameters(
    Index 0 Param 2=&RandomRangeFloat(Minimum=0, Maximum=4)
    Index 0 Param 4=2.5
    Module.Param0WriteEnabled=false
    Module.Param1WriteEnabled=true
    Module.Param2WriteEnabled=false
    Module.Param3WriteEnabled=true
   ) {module=DynamicMaterialParameters001}
   &ShapeLocation(
    MaskSpawnGroup=false
    Non Uniform Scale=<11 1 1>
    Offset=<5 0 0>
    Offset Coordinate Space=::Local
    Offset Mode=::NewEnumerator0
    Override Seed=false
    Sphere Radius=1
    UseOwnerScale=true
   )
  ]
  particleUpdateScript [
   &AccelerationForce(Acceleration=&UniformRangedVector(Minimum=<0 0 15>, Maximum=<0 0 40>))
   &ParticleState()
   &ScaleSpriteSize(Scale Factor=&Vector2DFromFloat(Value=&FloatFromCurve(
      FloatCurve=DataInterface() {propertyValues {Curve {keys [
          {arriveTangent=7.5764, interpMode=RCIM_Cubic, leaveTangent=7.5765, tangentMode=RCTM_User, time=0, value=0.02}
          {arriveTangent=0.0102, interpMode=RCIM_Cubic, leaveTangent=0.0102, tangentMode=RCTM_User, time=0.18, value=0.99}
          <at 1 0.1 RCIM_Cubic>
         ]}}}
      Scale Curve=1.2
     ))) {module=ScaleSpriteSize001}
   &DynamicMaterialParameters(
    Index 0 Param 1=&FloatFromCurve(FloatCurve=DataInterface() {propertyValues {Curve {keys [<at 0 1>]}}})
    Index 0 Param 2=0.5
    Index 0 Param 3=&FloatFromCurve(FloatCurve=DataInterface() {propertyValues {Curve {keys [
         <at 0 0.5 RCIM_Cubic>
         {arriveTangent=-0.0025, interpMode=RCIM_Cubic, leaveTangent=-0.0025, tangentMode=RCTM_User, time=0.6, value=0.625}
         {arriveTangent=2.7303, tangentMode=RCTM_User, time=1, value=2}
        ]}}})
    Module.Param0WriteEnabled=true
    Module.Param1WriteEnabled=false
    Module.Param2WriteEnabled=true
    Module.Param3WriteEnabled=false
   )
   &SpriteRotationRate(Rotation Rate=&Lerp_Float(A=-10, B=10, Alpha=$Particles.Initial.MaterialRandom)) {module=SpriteRotationRate001}
   /SetVariables(Particles.SubImageIndex=&FloatFromCurve(FloatCurve=DataInterface() {propertyValues {Curve {keys [<at 0 5>,<at 1 15>]}}})) {isSetParameters=true}
   &SolveForcesAndVelocity(Clamp Velocity=false, Limit Acceleration=false)
   &CameraOffset(Camera Offset Amount=5)
   &ScaleColor(
    Scale Alpha=&Multiply_Float(
     A=0.8
     B=&FloatFromCurve(FloatCurve=DataInterface() {propertyValues {Curve {keys [
          {arriveTangent=9.0309, interpMode=RCIM_Cubic, leaveTangent=9.0309, tangentMode=RCTM_User, time=0, value=0}
          {arriveTangent=-0.0116, interpMode=RCIM_Cubic, leaveTangent=-0.0116, tangentMode=RCTM_User, time=0.4073, value=1}
          {arriveTangent=-2.0743, interpMode=RCIM_Cubic, leaveTangent=-2.0743, tangentMode=RCTM_User, time=0.8, value=0.7}
          {interpMode=RCIM_Cubic, leaveTangent=-2.371, time=1, value=0}
         ]}}})
    )
    Scale RGB=&MakeVectorFromLinearColorRGB(Color=&ColorFromCurve(ColorCurve=DataInterface() {propertyValues {
        AlphaCurve {keys [<at 0.0141 0>,<at 0.2566 0.8>,<at 0.6705 0.5>,<at 1 0 RCIM_Cubic>]}
        BlueCurve {keys [<at 0 0 RCIM_Cubic>,<at 0.1098 0>,<at 0.3438 0>,<at 0.5257 0>,<at 0.6631 0.16>,<at 0.7992 0>,<at 1 0>]}
        GreenCurve {keys [<at 0 0 RCIM_Cubic>,<at 0.1098 0.1167>,<at 0.3438 0.4667>,<at 0.5257 2.5667>,<at 0.6631 5.176>,<at 0.7992 0.3667>,<at 1 0>]}
        RedCurve {keys [<at 0 1 RCIM_Cubic>,<at 0.1098 1>,<at 0.3438 4>,<at 0.5257 11>,<at 0.6631 16>,<at 0.7992 1>,<at 1 1>]}
       }}))
    ScaleA=true
    ScaleRGB=true
   )
  ]
  rendererClasses [&NiagaraSpriteRendererProperties]
  renderers [{rendererClass=&NiagaraSpriteRendererProperties}]
 }
 Fire_Sprites_SubUV001 {
  emitterUpdateScript [
   &EmitterState(Life Cycle Mode=::NewEnumerator1, Loop Duration=1, UseLoopDelay=false)
   &SpawnRate(SpawnRate=8, Use Spawn Probability=false)
  ]
  particleSpawnScript [
   &InitializeParticle(
    Color=<rgba 1 1 1 0.5>
    Lifetime=&UniformRangedFloat(Minimum=0.8, Maximum=1)
    Mass Mode=::NewEnumerator1
    Sprite Rotation Angle=&UniformRangedFloat(Minimum=-40, Maximum=40)
    Sprite Rotation Mode=::NewEnumerator0
    Sprite Size=&Vector2DFromFloat(Value=&UniformRangedFloat(Minimum=14, Maximum=20))
    Sprite Size Mode=::NewEnumerator4
    UsePositionOffset=false
    Write Material Random=true
   ) {module=InitializeParticle001}
   &ShapeLocation(
    MaskSpawnGroup=false
    Non Uniform Scale=<11 1 1>
    Offset=<5 0 0>
    Offset Coordinate Space=::Local
    Offset Mode=::NewEnumerator0
    Override Seed=false
    Sphere Radius=1
    UseOwnerScale=true
   ) {module=ShapeLocation001}
  ]
  particleUpdateScript [
   &ParticleState()
   &SubUVAnimation()
   &SolveForcesAndVelocity(Clamp Velocity=false, Limit Acceleration=false)
   &ScaleColor(
    Scale Alpha=&Multiply_Float(
     A=&FloatFromCurve(FloatCurve=DataInterface() {propertyValues {Curve {keys [
          {arriveTangent=16.5612, interpMode=RCIM_Cubic, leaveTangent=16.5612, leaveTangentWeight=0.3333, tangentMode=RCTM_User, time=0, value=0}
          {arriveTangentWeight=0.0333, interpMode=RCIM_Cubic, leaveTangentWeight=0.3, tangentMode=RCTM_User, time=0.1, value=1}
          {arriveTangentWeight=0.1333, interpMode=RCIM_Cubic, leaveTangentWeight=0.1667, tangentMode=RCTM_User, time=0.5, value=1}
          {arriveTangentWeight=0.1667, interpMode=RCIM_Cubic, tangentMode=RCTM_User, time=1, value=0}
         ]}}})
     B=&RandomRangeFloat(Minimum=0.5, Maximum=1)
    )
    Scale RGB=&Multiply_VectorByFloat(
     Vector=<1 1 1>
     Float=&FloatFromCurve(
      FloatCurve=DataInterface() {propertyValues {Curve {keys [
          {arriveTangent=9.0725, interpMode=RCIM_Cubic, leaveTangent=9.0725, tangentMode=RCTM_User, time=0, value=0}
          {arriveTangent=-2.5249, interpMode=RCIM_Cubic, leaveTangent=-2.5249, tangentMode=RCTM_User, time=0.1408, value=0.9847}
          {arriveTangent=-0.0405, interpMode=RCIM_Cubic, leaveTangent=-0.0405, tangentMode=RCTM_User, time=1, value=0}
         ]}}}
      Scale Curve=&RandomRangeFloat(Minimum=0.5, Maximum=1)
     )
    )
    ScaleA=true
    ScaleRGB=true
   )
   &ScaleSpriteSize(Scale Factor=&Vector2DFromFloat(Value=&FloatFromCurve(FloatCurve=DataInterface() {propertyValues {Curve {keys [
          <at 0 0.7 RCIM_Cubic>
          {arriveTangent=-0.565, interpMode=RCIM_Cubic, leaveTangent=-0.565, tangentMode=RCTM_User, time=0.2, value=1.5}
          {interpMode=RCIM_Cubic, tangentMode=RCTM_User, time=1, value=0.75}
         ]}}}))) {module=ScaleSpriteSize001}
   &DynamicMaterialParameters(
    Index 0 Param 1=&RandomRangeFloat(Minimum=2, Maximum=1)
    Index 0 Param 2=&RandomRangeFloat(Minimum=0, Maximum=2)
    Index 0 Param 3=8
    Index 0 Param 4=$Particles.SubImageIndex
    Index 1 Param 1=&RandomRangeFloat(Minimum=-1, Maximum=1)
    Index 1 Param 2=&RandomRangeFloat(Minimum=-1, Maximum=1)
    Index 1 Param 3=&FloatFromCurve(FloatCurve=DataInterface() {propertyValues {Curve {keys [
         {arriveTangent=4.1785, interpMode=RCIM_Cubic, leaveTangent=4.1785, tangentMode=RCTM_User, time=0, value=0}
         {arriveTangent=0.9995, interpMode=RCIM_Cubic, leaveTangent=0.9995, tangentMode=RCTM_User, time=0.2223, value=0.6825}
         {arriveTangent=-0.0405, interpMode=RCIM_Cubic, leaveTangent=-0.0405, tangentMode=RCTM_User, time=0.7, value=0.9}
         <at 1 0.7>
        ]}}})
    Index 2 Param 4=1
    Module.Param0WriteEnabled=true
    Module.Param0WriteEnabled1=true
    Module.Param0WriteEnabled2=false
    Module.Param1WriteEnabled=true
    Module.Param1WriteEnabled1=true
    Module.Param1WriteEnabled2=false
    Module.Param2WriteEnabled=true
    Module.Param2WriteEnabled1=true
    Module.Param2WriteEnabled2=false
    Module.Param3WriteEnabled=true
    Module.Param3WriteEnabled1=false
    Module.Param3WriteEnabled2=true
    Write Parameter Index 1=true
    Write Parameter Index 2=true
   )
   /SetVariables(Particles.Velocity=&TransformVector(Vector=<0 0 1>, Source Space=::World, Destination Space=::Local)) {isSetParameters=true}
   &CameraOffset(Camera Offset Amount=5)
  ]
  rendererClasses [&NiagaraSpriteRendererProperties]
  renderers [{rendererClass=&NiagaraSpriteRendererProperties}]
 }
 FlameCore {
  emitterUpdateScript [
   &EmitterState()
   &SpawnBurst_Instantaneous(Use Spawn Probability=false, UseLoopCountLimit=false)
  ]
  particleSpawnScript [
   &InitializeParticle(
    Lifetime=6
    Mesh Scale=&Multiply_VectorByFloat(Vector=<0.8 0.8 0.8>, Float=1.2)
    Mesh Scale Mode=::NewEnumerator4
    Sprite Rotation Mode=::NewEnumerator2
    UsePositionOffset=false
    Write Material Random=true
   )
   &InitialMeshOrientation(
    Mesh Orientation Mode=::NewEnumerator1
    Rotation=&RandomRangeVector(Minimum=<-1 0 0>, Maximum=<1 0 0>)
    Use Rotation Vector=true
   )
   &CameraOffset(Camera Offset Amount=8)
  ]
  particleUpdateScript [
   &ParticleState()
   &DynamicMaterialParameters(
    Index 0 Param 1=4
    Module.Index0Param0Write=true
    Module.Index0Param1Write=false
    Module.Index0Param2Write=false
    Module.Index0Param3Write=false
   )
   &SolveForcesAndVelocity(Clamp Velocity=false, Limit Acceleration=false)
   &ScaleColor(
    Scale Alpha=&Multiply_Float(
     A=&FloatFromCurve(FloatCurve=DataInterface() {propertyValues {Curve {keys [
          {interpMode=RCIM_Cubic, tangentMode=RCTM_User, time=0, value=0}
          {interpMode=RCIM_Cubic, tangentMode=RCTM_Break, time=0.1, value=1}
          {arriveTangent=0.0199, interpMode=RCIM_Cubic, leaveTangent=0.0199, tangentMode=RCTM_Break, time=0.8, value=1}
          <at 1 0 RCIM_Cubic>
         ]}}})
     B=&Multiply_Float(A=$System.FadeOut, B=$System.ToD)
    )
    Scale RGB=<2 2 2>
    ScaleA=true
    ScaleRGB=true
   ) {module=ScaleColor001}
  ]
  rendererClasses [&NiagaraMeshRendererProperties]
  renderers [{rendererClass=&NiagaraMeshRendererProperties, rendererIndex=0}]
 }
 Glow_Light {
  emitterUpdateScript [
   &EmitterState()
   &SpawnBurst_Instantaneous(Use Spawn Probability=false, UseLoopCountLimit=false)
  ]
  particleSpawnScript [
   &InitializeParticle(
    Color=<rgba 50 11.3817 0.7 1>
    Lifetime=6
    Position Offset=<3 0 0>
    Sprite Rotation Mode=::NewEnumerator2
    Sprite Size=<50 90>
    Sprite Size Mode=::NewEnumerator4
    UsePositionOffset=true
    Write Material Random=true
   )
   &SpriteFacingAndAlignment(Sprite Alignment=<1 0 0>, WriteAlignment=true, WriteFacing=true)
   &AddVelocity(Velocity=<0.1 0 0>)
   &CameraOffset(Camera Offset Amount=2)
  ]
  particleUpdateScript [
   &SolveForcesAndVelocity(Clamp Velocity=false, Limit Acceleration=false)
   &ParticleState()
   &ScaleColor(
    Scale Alpha=&Multiply_Float(
     A=&Waveform(
      Global Input Position=&RandomRangeFloat(Minimum=0, Maximum=1)
      "Waveform [1]"=::NewEnumerator0
      "[1] Frequency"=1
      "[1] Amplitude Min/Max"=<0.5 1>
      "[1] Scale"=&Waveform(
       Waveform Count=::NewEnumerator1
       Global Input Position=0.33
       "Waveform [1]"=::NewEnumerator0
       "[1] Frequency"=0.33
       "[1] Amplitude Min/Max"=<0.5 1.5>
       "[2] Average"=false
       "[2] Input Position"=0
       "[2] Phase"=$Emitter.Age
       "[2] Phase Scale"=1
       "[2] Compound Frequency"=2
       "[2] Frequency"=2
       "[2] Amplitude Min/Max"=<0 1>
       "[2] Scale"=1
       "[2] Bias"=0
       "[2] Exponent"=1
       "[2] Pulse Duration"=0.1
       "[2] Clip Triangle Min/Max"=<0 0>
       "[2] Random Seed"=0
      )
     )
     B=&Multiply_Float(A=$System.FadeOut, B=$System.ToD)
    )
    ScaleA=true
    ScaleRGB=true
   )
   /SetVariables(Particles.ColorGlow:LinearColor=&MultiplyLinearColor(
     Linear Color=$Particles.Color
     Scale Factor=&MakeVector4(
      X=1
      Y=1
      Z=1
      W=&Multiply_Float(
       A=0.005
       B=&Multiply_Float(A=$System.ToD, B=&FN_FrontEndFloatModifier(Gameplay=1, Front End=0.5))
      )
     )
    )) {isSetParameters=true}
   /SetVariables(Particles.LightColor=&MultiplyLinearColor(
     Linear Color=$Particles.Color
     Scale Factor=&MakeFloat4FromVectorAndFloat(
      "Vector (RGB)"=<1 1 1>
      "Float (Alpha)"=&FNTimeOfDayScalar(Night Scale=0.5, Morning Scale=1, Evening Scale=1)
     )
    )) {isSetParameters=true}
   &FN_Light_Attributes(
    Color=<rgba 1 0.2667 0 1>
    Create Color Attribute=true
    Create Exponent Attribute=true
    Create Radius Attribute=true
    Enable Particle Light Enable=true
    Exponent=3
    Light Enabled:NiagaraBool=&SetBoolByFloatComparison(A=&FN_FrontEndFloatModifier(Gameplay=1, Front End=0), B=1, Comparison Type=::NewEnumerator4)
    Radius=30
    Scale Color=&Multiply_Float(
     A=7
     B=&FloatFromCurve(FloatCurve=DataInterface() {propertyValues {Curve {keys [
          {interpMode=RCIM_Cubic, tangentMode=RCTM_Break, time=0.25, value=1}
          {interpMode=RCIM_Cubic, tangentMode=RCTM_Break, time=0.28, value=0.333}
          {interpMode=RCIM_Cubic, tangentMode=RCTM_Break, time=0.36, value=0.333}
          {interpMode=RCIM_Cubic, tangentMode=RCTM_Break, time=0.38, value=1}
          {interpMode=RCIM_Cubic, tangentMode=RCTM_Break, time=0.46, value=1}
          {interpMode=RCIM_Cubic, tangentMode=RCTM_Break, time=0.48, value=0.2666}
          {interpMode=RCIM_Cubic, tangentMode=RCTM_Break, time=0.5, value=0.2666}
          {interpMode=RCIM_Cubic, tangentMode=RCTM_Break, time=0.52, value=1}
         ]}}})
    )
   )
  ]
  rendererClasses [&NiagaraSpriteRendererProperties, &NiagaraLightRendererProperties]
  renderers [
   {rendererClass=&NiagaraSpriteRendererProperties}
   {rendererClass=&NiagaraLightRendererProperties, rendererIndex=1}
   {rendererClass=&NiagaraSpriteRendererProperties, rendererIndex=2}
  ]
 }
 Particle001 {
  emitterUpdateScript [
   &EmitterState(Life Cycle Mode=::NewEnumerator1, Loop Behavior=::NewEnumerator1, Loop Duration=5.6, UseLoopDelay=false)
   &SpawnPerUnit(
    Module.Use Max Movement Threshold=true
    Module.Use Movement Tolerance=false
    Spawn Probability:NiagaraFloat=&Multiply_Float(
     A=0.6
     B=&FloatFromCurve(
      FloatCurve=DataInterface() {propertyValues {Curve {keys [<at 0 0.5>,<at 0.2 1>,<at 0.4 0.7>,<at 0.5 0>,<at 0.55 0>,<at 0.6 1>,<at 0.8 1>,<at 1 0.5>]}}}
      CurveIndex=&Frac_Float(Float=&Add_Float(A=$Emitter.NormalizedLoopAge, B=&RandomRangeFloat(Minimum=0, Maximum=1)))
     )
    )
    Spawn Spacing=10
    Use Spawn Probability=true
   )
   &SpawnRate(
    Spawn Probability=&Multiply_Float(
     A=0.8
     B=&FloatFromCurve(
      FloatCurve=DataInterface() {propertyValues {Curve {keys [<at 0 1>,<at 0.2 1>,<at 0.4 0.7>,<at 0.5 0>,<at 0.55 0>,<at 0.6 1>,<at 0.8 1>,<at 1 0.5>]}}}
      CurveIndex=&Frac_Float(Float=&Add_Float(A=$Emitter.NormalizedLoopAge, B=&RandomRangeFloat(Minimum=0, Maximum=1)))
     )
    )
    SpawnRate=&Multiply_Float(
     A=30
     B=&FloatFromCurve(FloatCurve=DataInterface() {propertyValues {Curve {keys [<at 0.02 6>,<at 0.03 1>]}}}, CurveIndex=$Emitter.NormalizedLoopAge)
    )
    Use Spawn Probability=true
   )
  ]
  particleSpawnScript [
   &InitializeParticle(
    Color=&LerpLinearColors(
     StartColor=<rgba 10 1.0866 0 1>
     EndColor=<rgba 2 0.0333 0 1>
     LerpFactor=&UniformAOrBFloat(A=0, Distribution Weight=0.2, Evaluation Type=::NewEnumerator0, Recalculate Random Each Loop=true, Override Seed=false)
    )
    Lifetime Max=1.3
    Lifetime Min=0.3
    Lifetime Mode=::NewEnumerator1
    Sprite Size Mode=::NewEnumerator1
    Uniform Sprite Size Max=2
    Uniform Sprite Size Min=0.8
    UsePositionOffset=false
    Write Material Random=true
   )
   &ShapeLocation(
    Cylinder Height=10
    Cylinder Radius=6
    MaskSpawnGroup=false
    Offset=<7 0 0>
    Offset Coordinate Space=::Local
    Offset Mode=::NewEnumerator0
    Override Seed=false
    Rotation Mode=::NewEnumerator1
    Shape Primitive=::NewEnumerator1
    Surface Only=false
    UseOwnerScale=true
   )
   &AddVelocityFromPoint(Module.Use Falloff Distance=false, Origin Offset=<0 0 -50>, Origin Offset Coordinate Space=::World, Velocity Strength=70)
   &InheritVelocity(Inherited Velocity Amount Scale=<0.9 0.9 0.9>, Inherited Velocity Speed Limit=300, Speed Limit=true, UseSpeedThreshold=false)
  ]
  particleUpdateScript [
   &ParticleState()
   &AccelerationForce(Acceleration=&Multiply_Vector(
     A=&VectorFromFloat(Value=1)
     B=&UniformRangedVector(Minimum=<0 0 150>, Maximum=<0 0 250>)
    ))
   &Drag(
    Drag=&FloatFromCurve(FloatCurve=DataInterface() {propertyValues {Curve {keys [<at 0 0.5>,<at 0.5 0.5>,<at 1 1>]}}}, Scale Curve=9)
    Use Linear Drag=true
    Use Rotational Drag=false
   )
   /Niagara/Modules/Update/Forces/CurlNoiseForce(
    Noise Frequency:NiagaraFloat=&RandomRangeFloat(Minimum=0.05, Maximum=0.3)
    Noise Strength=&FloatFromCurve(FloatCurve=DataInterface() {propertyValues {Curve {keys [<at 0 1>,<at 1 0>]}}}, Scale Curve=500)
    Pan Noise=true
    Pan Noise Field=<0 0 1>
   )
   &ScaleSpriteSize(Scale Factor=&Vector2DFromFloat(Value=&FloatFromCurve(FloatCurve=DataInterface() {propertyValues {Curve {keys [
          {arriveTangent=9.1531, interpMode=RCIM_Cubic, leaveTangent=9.1531, tangentMode=RCTM_User, time=0, value=0.02}
          {arriveTangent=-0.02, interpMode=RCIM_Cubic, leaveTangent=-0.02, time=0.18, value=0.99}
          <at 1 0 RCIM_Cubic>
         ]}}}))) {module=ScaleSpriteSize002}
   &SolveForcesAndVelocity(Clamp Velocity=false, Limit Acceleration=false)
   &ScaleColor(
    Scale Alpha=&FloatFromCurve(
     FloatCurve=DataInterface() {propertyValues {Curve {keys [
         {arriveTangent=9.1531, interpMode=RCIM_Cubic, leaveTangent=9.1531, tangentMode=RCTM_User, time=0, value=0.02}
         {arriveTangent=0.48, interpMode=RCIM_Cubic, leaveTangent=0.48, time=0.18, value=0.99}
         <at 1 0.5 RCIM_Cubic>
        ]}}}
     Scale Curve=&MakeCustomFloatFromBool(Bool=$User.bIsFrontEndPreview, True Float=0.5, False Float=1)
    )
    Scale RGB=&MakeVectorFromLinearColorRGB(Color=&ColorFromCurve(ColorCurve=DataInterface() {propertyValues {
        AlphaCurve {keys [<at 0 1 RCIM_Cubic>]}
        BlueCurve {keys [<at 0 1.2255 RCIM_Cubic>,<at 0.6079 0 RCIM_Cubic>]}
        GreenCurve {keys [<at 0 3.7041 RCIM_Cubic>,<at 0.6079 0.0133 RCIM_Cubic>]}
        RedCurve {keys [<at 0 5 RCIM_Cubic>,<at 0.6079 0.8 RCIM_Cubic>]}
       }}))
    ScaleA=true
    ScaleRGB=true
   ) {module=ScaleColor001}
   &ScaleSpriteSizeBySpeed(
    Max Scale Factor=<1.5 5>
    Min Scale Factor=<0.4 0.8>
    Sample Scale Factor By Curve=true
    Scale Factor Curve=DataInterface() {propertyValues {Curve {keys [
        <at -0.0077 0.0765 RCIM_Cubic>
        {arriveTangent=1.2463, interpMode=RCIM_Cubic, leaveTangent=1.2463, tangentMode=RCTM_User, time=0.9985, value=0.9549}
       ]}}}
    Velocity Threshold=500
   ) {module=ScaleSpriteSizeBySpeed001}
   &FN_Sprite_Attributes(
    Color=&MultiplyLinearColor(Linear Color=$Particles.Color, Scale Factor=<1 1 1 0.005>)
    Create Color Attribute=true
    Create SpriteSize Attribute=true
    SpriteSize=&Multiply_Vector2D_ByFloat(Float=5, Vector2D=$Particles.SpriteSize)
    Write to Module Namespace=true
   )
  ]
  rendererClasses [&NiagaraSpriteRendererProperties]
  renderers [
   {rendererClass=&NiagaraSpriteRendererProperties}
   {rendererClass=&NiagaraSpriteRendererProperties, rendererIndex=1}
  ]
 }
 SoftTrail {
  emitterUpdateScript [
   &EmitterState()
   &FN_SpawnAnimTrail(AnimTrailWidthMode=::FromSecond, Number of Anim Trails=2)
  ]
  particleSpawnScript [
   &FN_InitializeAnimTrailBoneIndecies()
   &FN_InitializeAnimTrailBoneIndecies(AnimTrail Selection=1, First Bone or Socket Index=2, Second Bone or Socket Index=3) {module=FN_InitializeAnimTrailBoneIndecies001}
   &FN_InitializeAnimTrail(AnimTrailWidth=14, Lifetime=0.2, SkeletalMesh=$System.Skeletal Mesh)
  ]
  particleUpdateScript [
   &FN_UpdateAnimTrail(SkeletalMesh=$System.Skeletal Mesh) {module=FN_UpdateAnimTrail001}
   &Color(Color=&ColorFromCurve(ColorCurve=DataInterface() {propertyValues {
       AlphaCurve {keys [<at 0 1 RCIM_Cubic>,<at 1 0 RCIM_Cubic>]}
       BlueCurve {keys [<at 0 0>,<at 0.416 0.1>]}
       GreenCurve {keys [<at 0 1.5>,<at 0.416 0.1633>]}
       RedCurve {keys [<at 0 10>,<at 0.416 2>]}
      }})) {module=Color001}
   &ScaleColor(
    Color Value To Scale=$Particles.Color
    Scale Alpha=&Multiply_Float(
     A=$System.FadeOut
     B=&FloatFromCurve(FloatCurve=DataInterface() {propertyValues {Curve {keys [{arriveTangent=-1.3422, interpMode=RCIM_Cubic, leaveTangent=-1.3422, tangentMode=RCTM_User, time=0, value=1}, <at 1 0 RCIM_Cubic>]}}})
    )
    ScaleA=true
    ScaleRGB=false
   )
   &GravityForce(Gravity=<0 0 980>)
   &SolveForcesAndVelocity(Clamp Velocity=false, Limit Acceleration=false)
   &DynamicMaterialParameters(
    Index 0 Param 1=20
    Index 0 Param 2=1
    Index 0 Param 4=&FloatFromCurve(
     FloatCurve=DataInterface() {propertyValues {Curve {keys [
         {arriveTangent=2.1132, interpMode=RCIM_Cubic, leaveTangent=2.1132, tangentMode=RCTM_User, time=0.1, value=0}
         {interpMode=RCIM_Cubic, tangentMode=RCTM_User, time=1, value=1}
        ]}}}
     Scale Curve=10
    )
    Module.Index0Param0Write=false
    Module.Index0Param1Write=false
    Module.Index0Param2Write=false
    Module.Index0Param3Write=false
   ) {module=DynamicMaterialParameters001}
   &ScaleRibbonWidth(Ribbon Width Scale=&FloatFromCurve(FloatCurve=DataInterface() {propertyValues {Curve {keys [{arriveTangent=-1.6099, interpMode=RCIM_Cubic, leaveTangent=-1.6099, tangentMode=RCTM_User, time=0, value=1}, <at 1 0 RCIM_Cubic>]}}}))
  ]
  rendererClasses [&NiagaraRibbonRendererProperties]
  renderers [{rendererClass=&NiagaraRibbonRendererProperties}]
 }
}
summary {
 systemName=NS_StrategicSpur_Emote_FireStick_L
 emitters [
  {emitterName=Glow_Light, rendererClasses [&NiagaraSpriteRendererProperties, &NiagaraLightRendererProperties]}
  {emitterName=FlameCore, rendererClasses [&NiagaraMeshRendererProperties]}
  {emitterName=SoftTrail, rendererClasses [&NiagaraRibbonRendererProperties]}
  {emitterName=Particle001, rendererClasses [&NiagaraSpriteRendererProperties]}
  {emitterName=FireLicks_Sprites_Tall001, rendererClasses [&NiagaraSpriteRendererProperties]}
  {emitterName=Fire_Sprites_SubUV001, rendererClasses [&NiagaraSpriteRendererProperties]}
  {bEnabled=false, emitterName=FN_Debug, rendererClasses [&NiagaraSpriteRendererProperties]}
 ]
 userVariables [
  User.bIsFrontEnd {type=NiagaraBool}
  User.bIsFrontEndPreview {type=NiagaraBool}
  User.Reveal {type=NiagaraFloat, defaultValue=0, description=""}
  User.RibbonIDOffset {type=NiagaraInt32}
 ]
}
