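# One-shot Niagara ignition burst for the fire-sticks moment of the StrategicSpur emote.
# LightFlash adds a glow sprite and point light, while Scorch layers fading sprites from one locally oriented particle.
# Sparks rise with drag, curl noise and speed-based stretching, and ring-spawned smoke swirls outward, cooling from orange to dark grey.
# Time-of-day scalars dim flash and scorch, and front-end modifiers tone down brightness and disable the light.
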NiagaraSystem /BRCosmetics/Animation/Game/MainPlayer/Emotes/StrategicSpur/FX/NS_StrategicSpur_Emote_FireSticks_Ignition
dependencies {
 usedDataInterfaces [&NiagaraDataInterfaceColorCurve, &NiagaraDataInterfaceCurve, &NiagaraDataInterfacePlatformSet]
 usedDynamicInputs []
 usedModules [
  &SystemState
  &self:SystemSpawnScript.SystemScriptSource.SystemScriptGraph.NiagaraNodeAssignment_0.SetVariables_#1
 ]
}
emitters {
 LightFlash {
  emitterUpdateScript [
   &EmitterState()
   &SpawnBurst_Instantaneous(Use Spawn Probability=false, UseLoopCountLimit=false)
  ]
  particleSpawnScript [
   &InitializeParticle(
    Lifetime=0.4
    Position Offset=<2 0 6>
    Sprite Rotation Mode=::NewEnumerator2
    Sprite Size Mode=::NewEnumerator3
    Uniform Sprite Size=100
    UsePositionOffset=true
    Write Material Random=true
   )
   &CameraOffset(Camera Offset Amount=5)
  ]
  particleUpdateScript [
   &ParticleState()
   &ScaleSpriteSize(Uniform Curve Sprite Scale=DataInterface() {propertyValues {Curve {keys [
        {arriveTangent=0.7118, interpMode=RCIM_Cubic, leaveTangent=0.7118, tangentMode=RCTM_User, time=0, value=0.7}
        {interpMode=RCIM_Cubic, tangentMode=RCTM_User, time=0.9, value=1}
       ]}}})
   &ScaleColor(
    Scale Alpha=&FloatFromCurve(
     FloatCurve=DataInterface() {propertyValues {Curve {keys [
         {arriveTangent=-2.6629, interpMode=RCIM_Cubic, leaveTangent=-2.6629, tangentMode=RCTM_User, time=0, value=1}
         {arriveTangent=-0.0145, tangentMode=RCTM_User, time=1, value=0}
        ]}}}
     Scale Curve=&Multiply_Float(
      A=&FNTimeOfDayScalar(Night Scale=0.5, Morning Scale=0.6, Evening Scale=0.6)
      B=&FN_FrontEndFloatModifier(Gameplay=1, Front End=0.2)
     )
    )
    Scale RGB=&MakeVectorFromLinearColorRGB(Color=&ColorFromCurve(ColorCurve=DataInterface() {propertyValues {
        AlphaCurve {keys [<at 0.1221 1 RCIM_Cubic>,<at 1 0 RCIM_Cubic>]}
        BlueCurve {keys [<at 0.1654 1.6 RCIM_Cubic>,<at 0.2627 0.4>]}
        GreenCurve {keys [<at 0.1654 17.28 RCIM_Cubic>,<at 0.2627 8.32>]}
        RedCurve {keys [<at 0.1654 80 RCIM_Cubic>,<at 0.2627 40>]}
       }}))
    ScaleA=true
    ScaleRGB=true
   )
   &FN_Light_Attributes(
    Color=&ColorFromCurve(ColorCurve=DataInterface() {propertyValues {
       AlphaCurve {keys [<at 0 1 RCIM_Cubic>,<at 1 0 RCIM_Cubic>]}
       BlueCurve {keys [<at 0.1528 0.5 RCIM_Cubic>,<at 0.5279 0.02>]}
       GreenCurve {keys [<at 0.1528 3.5083 RCIM_Cubic>,<at 0.5279 0.1017>]}
       RedCurve {keys [<at 0.1528 10 RCIM_Cubic>,<at 0.5279 1>]}
      }})
    Create Color Attribute=true
    Create Exponent Attribute=true
    Create Position Attribute=true
    Create Radius Attribute=true
    Enable Particle Light Enable=true
    Exponent=12
    Light Enabled:NiagaraBool=&SetBoolByFloatComparison(A=&FN_FrontEndFloatModifier(Gameplay=1, Front End=0), B=1, Comparison Type=::NewEnumerator4)
    Position=<0 0 10>
    Radius=100
    Scale Color=&Multiply_Float(A=6, B=&FNTimeOfDayScalar(Night Scale=0.5, Morning Scale=0.6, Evening Scale=0.6))
   )
  ]
  rendererClasses [&NiagaraLightRendererProperties, &NiagaraSpriteRendererProperties]
  renderers [
   {rendererClass=&NiagaraLightRendererProperties, rendererIndex=0}
   {rendererClass=&NiagaraSpriteRendererProperties, rendererIndex=1}
  ]
 }
 Particle {
  emitterUpdateScript [
   &EmitterState()
   &SpawnBurst_Instantaneous(
    Spawn Count=&RandomRangeInt(Minimum=17, Maximum=20)
    Use Spawn Probability=false
    UseLoopCountLimit=false
   )
  ]
  particleSpawnScript [
   &InitializeParticle(
    Color=&LerpLinearColors(
     StartColor=<rgba 10 1.0866 0 1>
     EndColor=<rgba 2 0.0333 0 1>
     LerpFactor=&UniformAOrBFloat(A=0, Distribution Weight=0.2, Evaluation Type=::NewEnumerator0, Recalculate Random Each Loop=true, Override Seed=false)
    )
    Lifetime Max=1
    Lifetime Min=0.5
    Lifetime Mode=::NewEnumerator1
    Sprite Size Mode=::NewEnumerator1
    Uniform Sprite Size Max=3
    Uniform Sprite Size Min=0.8
    UsePositionOffset=false
    Write Material Random=true
   )
   &ShapeLocation(MaskSpawnGroup=false, Offset Mode=::NewEnumerator0, Override Seed=false, Rotation Mode=::NewEnumerator1, Sphere Radius=1, UseOwnerScale=true)
   &AddVelocityFromPoint(
    Module.Use Falloff Distance=false
    Origin Offset=<0 0 -1>
    Velocity Strength=&RandomRangeFloat(Minimum=50, Maximum=400)
   )
   &InheritVelocity(Inherited Velocity Amount Scale=<0.9 0.9 0.9>, Inherited Velocity Speed Limit=300, Speed Limit=true, UseSpeedThreshold=false)
  ]
  particleUpdateScript [
   &ParticleState()
   /SetVariables(Particles.GlobalSize:NiagaraFloat=&FloatFromCurve(
     FloatCurve=DataInterface() {propertyValues {Curve {keys [{interpMode=RCIM_Cubic, tangentMode=RCTM_User, time=0, value=1}, {interpMode=RCIM_Cubic, tangentMode=RCTM_User, time=1, value=0.2}]}}}
     Scale Curve=&RandomRangeFloat(Minimum=0.7, Maximum=1)
    )) {isSetParameters=true}
   &AccelerationForce(Acceleration=&Multiply_Vector(
     A=&VectorFromFloat(Value=1)
     B=&UniformRangedVector(Minimum=<0 0 150>, Maximum=<0 0 250>)
    ))
   &Drag(
    Drag=&FloatFromCurve(
     FloatCurve=DataInterface() {propertyValues {Curve {keys [<at 0 0.5>,<at 0.5 0.5>,<at 1 1>]}}}
     Scale Curve=&RandomRangeFloat(Minimum=6, Maximum=18)
    )
    Use Linear Drag=true
    Use Rotational Drag=false
   )
   /Niagara/Modules/Update/Forces/CurlNoiseForce(
    Noise Frequency:NiagaraFloat=&RandomRangeFloat(Minimum=0.05, Maximum=0.3)
    Noise Strength=&FloatFromCurve(FloatCurve=DataInterface() {propertyValues {Curve {keys [<at 0 1>,<at 1 0>]}}}, Scale Curve=800)
    Pan Noise=true
    Pan Noise Field=<0 0 1>
   )
   &ScaleSpriteSize(Scale Factor=&Vector2DFromFloat(Value=&FloatFromCurve(FloatCurve=DataInterface() {propertyValues {Curve {keys [
          {arriveTangent=9.1531, interpMode=RCIM_Cubic, leaveTangent=9.1531, tangentMode=RCTM_User, time=0, value=0.02}
          {arriveTangent=-0.02, interpMode=RCIM_Cubic, leaveTangent=-0.02, time=0.18, value=0.99}
          <at 1 0 RCIM_Cubic>
         ]}}}))) {module=ScaleSpriteSize002}
   &SolveForcesAndVelocity(Clamp Velocity=false, Limit Acceleration=false)
   &ScaleColor(
    Scale Alpha=&FloatFromCurve(
     FloatCurve=DataInterface() {propertyValues {Curve {keys [
         {arriveTangent=9.1531, interpMode=RCIM_Cubic, leaveTangent=9.1531, tangentMode=RCTM_User, time=0, value=0.02}
         {arriveTangent=0.48, interpMode=RCIM_Cubic, leaveTangent=0.48, time=0.18, value=0.99}
         <at 1 0.5 RCIM_Cubic>
        ]}}}
     Scale Curve=&MakeCustomFloatFromBool(Bool=$User.bIsFrontEndPreview, True Float=0.5, False Float=1)
    )
    Scale RGB=&MakeVectorFromLinearColorRGB(Color=&ColorFromCurve(ColorCurve=DataInterface() {propertyValues {
        AlphaCurve {keys [<at 0 1 RCIM_Cubic>]}
        BlueCurve {keys [<at 0 1.2255 RCIM_Cubic>,<at 0.5609 0 RCIM_Cubic>]}
        GreenCurve {keys [<at 0 3.7041 RCIM_Cubic>,<at 0.5609 0.0133 RCIM_Cubic>]}
        RedCurve {keys [<at 0 5 RCIM_Cubic>,<at 0.5609 0.8 RCIM_Cubic>]}
       }}))
    ScaleA=true
    ScaleRGB=true
   )
   &ScaleSpriteSizeBySpeed(
    Max Scale Factor=<1.5 5>
    Min Scale Factor=<0.4 0.8>
    Sample Scale Factor By Curve=true
    Scale Factor Curve=DataInterface() {propertyValues {Curve {keys [
        <at -0.0077 0.0765 RCIM_Cubic>
        {arriveTangent=1.2463, interpMode=RCIM_Cubic, leaveTangent=1.2463, tangentMode=RCTM_User, time=0.9985, value=0.9549}
       ]}}}
    Velocity Threshold=500
   ) {module=ScaleSpriteSizeBySpeed001}
   &FN_Sprite_Attributes(
    Color=&MultiplyLinearColor(Linear Color=$Particles.Color, Scale Factor=<1 1 1 0.01>)
    Create Color Attribute=true
    Create SpriteSize Attribute=true
    SpriteSize=&Multiply_Vector2D_ByFloat(Float=5, Vector2D=$Particles.SpriteSize)
    Write to Module Namespace=true
   )
  ]
  rendererClasses [&NiagaraSpriteRendererProperties]
  renderers [
   {rendererClass=&NiagaraSpriteRendererProperties}
   {rendererClass=&NiagaraSpriteRendererProperties, rendererIndex=1}
  ]
 }
 Scorch {
  emitterUpdateScript [
   &EmitterState()
   &SpawnBurst_Instantaneous(Use Spawn Probability=false, UseLoopCountLimit=false)
  ]
  particleSpawnScript [
   &InitializeParticle(
    Color=<rgba 40 5.3333 0 1>
    Lifetime=&RandomRangeFloat(Minimum=0.888, Maximum=0.999)
    Mesh Scale Mode=::NewEnumerator3
    Sprite Rotation Mode=::NewEnumerator2
    Sprite Size Mode=::NewEnumerator3
    Uniform Sprite Size=30
    UsePositionOffset=false
    Write Material Random=true
   )
   &SpriteFacingAndAlignment(
    Alignment Coordinate Space=::Local
    Facing Coordinate Space=::Local
    Sprite Alignment=<1 0 0>
    Sprite Facing=<0 0 1>
    WriteAlignment=true
    WriteFacing=true
   )
   &CameraOffset(Camera Offset Amount=10)
  ]
  particleUpdateScript [
   &ParticleState()
   &FN_Sprite_Attributes(
    Color=&MultiplyLinearColor(
     Linear Color=&ColorFromCurve(ColorCurve=DataInterface() {propertyValues {
        AlphaCurve {keys [<at 0 0>, <at 0.0664 1>, {arriveTangent=-1.0712, interpMode=RCIM_Cubic, leaveTangent=-1.0712, time=0.4195, value=1}, <at 1 0 RCIM_Cubic>]}
        BlueCurve {keys [<at 0.1288 0 RCIM_Cubic>,<at 0.6154 0>]}
        GreenCurve {keys [<at 0.1288 0.4167 RCIM_Cubic>,<at 0.6154 0>]}
        RedCurve {keys [<at 0.1288 2.5 RCIM_Cubic>,<at 0.6154 0>]}
       }})
     Scale Factor=&MakeFloat4FromVectorAndFloat("Vector (RGB)"=<1 1 1>, "Float (Alpha)"=$System.ToD)
    )
    Create Color Attribute=true
    Create SpriteSize Attribute=true
    SpriteSize=&Vector2DFromFloat(Value=45)
    Write to Module Namespace=true
   )
   &FN_Sprite_Attributes(
    Camera Offset=3
    Color=&MultiplyLinearColor(
     Linear Color=&ColorFromCurve(ColorCurve=DataInterface() {propertyValues {
        AlphaCurve {keys [<at 0 0>, <at 0.0414 2>, {arriveTangent=-2.3345, interpMode=RCIM_Cubic, leaveTangent=-2.3345, time=0.396, value=2}, <at 0.8981 0 RCIM_Cubic>]}
        BlueCurve {keys [
          <at 0.0288 0.1468>
          <at 0.0863 0>
          {interpMode=RCIM_Cubic, tangentMode=RCTM_Break, time=0.1803, value=0}
          {interpMode=RCIM_Cubic, tangentMode=RCTM_Break, time=0.4751, value=0.025}
          <at 0.8585 0.0061>
         ]}
        GreenCurve {keys [
          <at 0.0288 0.4312>
          <at 0.0863 0.1667>
          {arriveTangent=-0.3515, interpMode=RCIM_Cubic, leaveTangent=-0.3515, time=0.1803, value=0.2667}
          {arriveTangent=-0.0427, interpMode=RCIM_Cubic, leaveTangent=-0.0427, tangentMode=RCTM_User, time=0.4751, value=0.03}
          <at 0.8585 0.0077>
         ]}
        RedCurve {keys [
          <at 0.0288 1>
          <at 0.0863 1>
          {arriveTangent=0.1102, interpMode=RCIM_Cubic, leaveTangent=0.1102, tangentMode=RCTM_User, time=0.1803, value=1}
          {arriveTangent=-0.113, interpMode=RCIM_Cubic, leaveTangent=-0.113, tangentMode=RCTM_User, time=0.4751, value=0.1}
          <at 0.8585 0.02>
         ]}
       }})
     Scale Factor=&MakeFloat4FromVectorAndFloat(
      "Vector (RGB)"=&VectorFromFloat(Value=&FN_FrontEndFloatModifier(Gameplay=1, Front End=0.2))
      "Float (Alpha)"=&Multiply_Float(A=$System.ToD, B=1)
     )
    )
    Create Camera Offset Attribute=true
    Create Color Attribute=true
    Create SpriteSize Attribute=true
    SpriteSize=&Vector2DFromFloat(Value=100)
    Write to Module Namespace=true
   ) {module=FN_Sprite_Attributes001}
   &ScaleColor(
    Scale Alpha=&FloatFromCurve(
     FloatCurve=DataInterface() {propertyValues {Curve {keys [
         <at 0 0>
         {interpMode=RCIM_Cubic, tangentMode=RCTM_User, time=0.1, value=1}
         {arriveTangent=-1.4286, interpMode=RCIM_Cubic, leaveTangent=-1.4286, time=0.4, value=1}
         {interpMode=RCIM_Cubic, tangentMode=RCTM_User, time=0.8, value=0}
        ]}}}
     Scale Curve=&Multiply_Float(A=$System.ToD, B=&FN_FrontEndFloatModifier(Gameplay=1, Front End=0.1))
    )
    Scale RGB=&MakeVectorFromLinearColorRGB(Color=&ColorFromCurve(ColorCurve=DataInterface() {propertyValues {
        AlphaCurve {keys [<at 0 1 RCIM_Cubic>,<at 0.6443 0 RCIM_Cubic>]}
        BlueCurve {keys [<at 0.0752 3.2 RCIM_Cubic>,<at 0.244 0>,<at 0.4332 0>]}
        GreenCurve {keys [<at 0.0752 3.4667 RCIM_Cubic>,<at 0.244 0.3667>,<at 0.4332 0>]}
        RedCurve {keys [<at 0.0752 4 RCIM_Cubic>,<at 0.244 1>,<at 0.4332 0>]}
       }}))
    ScaleA=true
    ScaleRGB=true
   )
   &FN_Sprite_Attributes(
    Camera Offset:NiagaraFloat=&FN_PickFloatDependOnPlatform(TargetPlatform=DataInterface() {propertyValues {Platforms {qualityLevelMask=-2}}}, TargetPlatformActive=10, TargetPlatformNoActive=40)
    Color=&MakeLinearColorFromVectorAndFloat(
     "Vector (RGB)"=&MakeVectorFromLinearColorRGB(Color=<rgba 100 8.3333 0 1>)
     "Float (Alpha)"=&FloatFromCurve(
      FloatCurve=DataInterface() {propertyValues {Curve {keys [<at 0 1>,<at 0.1 0>]}}}
      Scale Curve=&Multiply_Float(A=$System.ToD, B=1)
     )
    )
    Create Camera Offset Attribute=true
    Create Color Attribute=true
    Create SpriteSize Attribute=true
    SpriteSize=&Vector2DFromFloat(Value=50)
    Write to Module Namespace=true
   ) {module=FN_Sprite_Attributes003}
   &FN_Sprite_Attributes(
    Color=&MultiplyLinearColor(
     Linear Color=<rgba 4 0.3333 0 0.5>
     Scale Factor=&MakeVector4(
      X=1
      Y=1
      Z=1
      W=&FloatFromCurve(FloatCurve=DataInterface() {propertyValues {Curve {keys [<at 0 1>,<at 0.3 0>]}}}, Scale Curve=$System.ToD)
     )
    )
    Create Color Attribute=true
    Create SpriteSize Attribute=true
    SpriteSize=&Multiply_Vector2D_ByFloat(Float=0.2, Vector2D=<200 200>)
    Write to Module Namespace=true
   ) {module=FN_Sprite_Attributes002}
   &FN_Sprite_Attributes(
    Color=&MultiplyLinearColor(
     Linear Color=<rgba 0.04 0.04 0.04 1.3>
     Scale Factor=&MakeVector4(
      X=1
      Y=1
      Z=1
      W=&FloatFromCurve(FloatCurve=DataInterface() {propertyValues {Curve {keys [<at 0 1>,<at 1 0>]}}}, Scale Curve=$System.ToD)
     )
    )
    Create Color Attribute=true
    Create SpriteSize Attribute=true
    SpriteSize=&Multiply_Vector2D_ByFloat(Float=0.5, Vector2D=<220 200>)
    Write to Module Namespace=true
   ) {module=FN_Sprite_Attributes004}
  ]
  rendererClasses [&NiagaraSpriteRendererProperties]
  renderers [
   {rendererClass=&NiagaraSpriteRendererProperties}
   {rendererClass=&NiagaraSpriteRendererProperties, rendererIndex=1}
   {rendererClass=&NiagaraSpriteRendererProperties, rendererIndex=2}
   {rendererClass=&NiagaraSpriteRendererProperties, rendererIndex=3}
   {rendererClass=&NiagaraSpriteRendererProperties, rendererIndex=4}
   {rendererClass=&NiagaraSpriteRendererProperties, rendererIndex=5}
  ]
 }
 SmokeLargeLocalSpace {
  emitterUpdateScript [
   &EmitterState()
   &SpawnBurst_Instantaneous(Spawn Count=20, Use Spawn Probability=false, UseLoopCountLimit=false)
  ]
  particleSpawnScript [
   &InitializeParticle(
    Color=&MultiplyLinearColor(
     Linear Color=&RandomRangeLinearColor(Minimum=<rgba 0.5781 0.5781 0.5781 0.233>, Maximum=<rgba 1 1 1 0.5>)
     Scale Factor=&MakeVector4(X=0, Y=0, Z=0, W=&FN_FrontEndFloatModifier(Gameplay=1, Front End=0.5))
    )
    Lifetime=&RandomRangeFloat(Minimum=0.25, Maximum=0.6)
    Sprite Rotation Mode=::NewEnumerator2
    Sprite Size Mode=::NewEnumerator3
    Uniform Sprite Size=&RandomRangeFloat(Minimum=15, Maximum=23)
    UsePositionOffset=true
    Write Material Random=true
   )
   &ShapeLocation(MaskSpawnGroup=false, Override Seed=false, Ring Radius=10, Shape Primitive=::NewEnumerator4, UseOwnerScale=true)
   &AddVelocity(Velocity=&Multiply_VectorByFloat(
     Vector=&AddVectorToPosition(Position=Linked() {linkedVariable {name=Particles.ShapeLocation.ShapeTangent, type=Vector3f}}, Vector=<0 0 1>)
     Float=70
    ))
  ]
  particleUpdateScript [
   &Drag(Drag=2.2, Use Linear Drag=true, Use Rotational Drag=true)
   &SolveForcesAndVelocity(Clamp Velocity=false, Limit Acceleration=false)
   &ParticleState()
   &SpriteRotationRate(Rotation Rate=200)
   &Color(
    Color=&ColorFromCurve(ColorCurve=DataInterface() {propertyValues {
       AlphaCurve {keys [<at 0 0.2>,<at 0.1424 1>,<at 0.5109 0.666>,<at 1 0>]}
       BlueCurve {keys [<at 0 0>, {arriveTangent=0.0577, time=0.1578, value=0.075}, <at 1 0.05>]}
       GreenCurve {keys [<at 0 0.1833>, {arriveTangent=-0.0961, time=0.1578, value=0.075}, <at 1 0.05>]}
       RedCurve {keys [<at 0 1>, {arriveTangent=-1.0957, time=0.1578, value=0.075}, <at 1 0.05>]}
      }})
    Scale Alpha=&MakeFloatFromLinearColor(LinearColor=$Particles.Initial.Color, Channel:ENiagaraLinearColor_Channels=::NewEnumerator3)
   )
   &ScaleSpriteSize(
    Uniform Curve Scale=&RandomRangeFloat(Minimum=0.8, Maximum=1.1)
    Uniform Curve Sprite Scale=DataInterface() {propertyValues {Curve {keys [
        {interpMode=RCIM_Cubic, tangentMode=RCTM_User, time=0, value=0}
        {interpMode=RCIM_Cubic, leaveTangent=1.5237, tangentMode=RCTM_User, time=0.1, value=0.6}
        {interpMode=RCIM_Cubic, tangentMode=RCTM_User, time=0.6, value=1.1}
       ]}}}
   )
   &DynamicMaterialParameters(
    Index 0 Param 1=&Add_Float(
     A=&Multiply_Float(A=$Engine.Time, B=1)
     B=&RandomRangeFloat(Minimum=0, Maximum=1)
    )
    Index 0 Param 3=&FloatFromCurve(FloatCurve=DataInterface() {propertyValues {Curve {keys [
         {arriveTangent=-3.5161, interpMode=RCIM_Cubic, leaveTangent=-3.5161, tangentMode=RCTM_User, time=0, value=1}
         {interpMode=RCIM_Cubic, tangentMode=RCTM_User, time=0.3, value=0}
        ]}}})
    Index 2 Param 1=1
    Index 2 Param 2=&RandomRangeFloat(Minimum=0, Maximum=1)
    Index 2 Param 3=&FloatFromCurve(FloatCurve=DataInterface() {propertyValues {Curve {keys [
         {interpMode=RCIM_Cubic, tangentMode=RCTM_Break, time=0.666, value=0}
         {arriveTangent=10.9325, interpMode=RCIM_Cubic, tangentMode=RCTM_Break, time=1, value=2}
        ]}}})
    Index 2 Param 4=1
    Module.Param0WriteEnabled=true
    Module.Param0WriteEnabled1=true
    Module.Param0WriteEnabled2=true
    Module.Param1WriteEnabled=true
    Module.Param1WriteEnabled1=true
    Module.Param1WriteEnabled2=true
    Module.Param2WriteEnabled=true
    Module.Param2WriteEnabled1=true
    Module.Param2WriteEnabled2=true
    Module.Param3WriteEnabled=true
    Module.Param3WriteEnabled1=true
    Module.Param3WriteEnabled2=true
    Write Parameter Index 1=true
    Write Parameter Index 2=true
   )
  ]
  rendererClasses [&NiagaraSpriteRendererProperties]
  renderers [
   {rendererClass=&NiagaraSpriteRendererProperties}
   {rendererClass=&NiagaraSpriteRendererProperties, rendererIndex=1}
  ]
 }
}
summary {
 systemName=NS_StrategicSpur_Emote_FireSticks_Ignition
 emitters [
  {emitterName=Scorch, rendererClasses [&NiagaraSpriteRendererProperties]}
  {emitterName=LightFlash, rendererClasses [&NiagaraLightRendererProperties, &NiagaraSpriteRendererProperties]}
  {emitterName=SmokeLargeLocalSpace, rendererClasses [&NiagaraSpriteRendererProperties]}
  {emitterName=Particle, rendererClasses [&NiagaraSpriteRendererProperties]}
 ]
 userVariables [
  User.bIsFrontEnd {type=NiagaraBool}
  User.bIsFrontEndPreview {type=NiagaraBool}
 ]
}
